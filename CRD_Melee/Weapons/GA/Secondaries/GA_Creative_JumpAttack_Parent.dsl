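; Parent gameplay ability for the creative melee jump attack: a leap and ground slam.
; Activation checks weapon support and spends energy, then launches the player, optionally with aim-assist targeting, and starts a fall montage.
; A timed downward trace predicts touchdown for animation; landing triggers targeted effects, knockback and slam cue before net-synced ending.
; Swimming, skydiving or clambering cancels it, suppressing the slam cue; DoOnce gates are reset each activation.
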(event Custom|Predict
  ;[1 1 1 1 64] PredictionWindow
  (bind _returnvalue (Utilities|Time|SetTimerbyEvent (delegate Custom|PredictUpdate) (Variables|Const|GetPredictionUpdateRateInSeconds)))
  (Variables|Runtime|SetPredictLandingTimerHandle _returnvalue)
  (EventDispatchers|AssignOnPawnLanded (Variables|Runtime|GetOut%20Fort%20Player%20Pawn) (delegate Custom|FallbackLandEvent))
  (CallFunction|WaitforLand))

(event Custom|CancelIfClambering
  ;[1 1 1] Listen for Clamber
  (bind _asynctaskproxy (Ability|Tasks|WaitGameplayTagAdd "(TagName=\"Clambering.State.LedgeMoving\")" (Ability|GetAvatarActorfromActorInfo))
    (:then
      (Variables|Default|SetClamberingStartTask _asynctaskproxy))
    (:Added
      (Variables|Default|SetbIgnoreLanded true)
      (Ability|CancelAbility))))

(event Custom|FallbackLandEvent
  (params (Hit :type Hit%20Result%20Structure%20%28by%20ref%29))
  ;[1 1 1 1 64] PredictionWindow
  (EventDispatchers|UnbindEventfromOnPawnLanded (Variables|Runtime|GetOut%20Fort%20Player%20Pawn) (delegate Custom|FallbackLandEvent))
  (Utilities|Time|ClearandInvalidateTimerbyHandle (Variables|Runtime|GetPredictLandingTimerHandle))
  (CallFunction|PredictionHit))

(event Custom|UnbindFallbackLandCheck
  ;[1 1 1 1 64] PredictionWindow
  (EventDispatchers|UnbindEventfromOnPawnLanded (Variables|Runtime|GetOut%20Fort%20Player%20Pawn) (delegate Custom|FallbackLandEvent))
  (Utilities|Time|ClearandInvalidateTimerbyHandle (Variables|Runtime|GetPredictLandingTimerHandle))
  (CallFunction|PredictionHit))

(event Custom|PredictUpdate
  (bind (_min _max) (CallFunction|PredictioninRangeSeconds))
  ;[1 1 1 1 64] PredictionWindow
  (if (CallFunction|IsAbilityPlayerFalling)
    (bind (_outhit _returnvalue) (Collision|LineTraceByChannel (CallFunction|GetPlayerLocationwithCapsuleOffset) (+ (CallFunction|GetPlayerLocationwithCapsuleOffset) (* (* (Transformation|GetActorUpVector (Variables|Runtime|GetOut%20Fort%20Player%20Pawn)) -1.0) 1e+04)) "TraceTypeQuery10" false (Utilities|Array|MakeArray (Variables|Runtime|GetOut%20Fort%20Player%20Pawn))))
    (if _returnvalue
      (if (or (and (< (CallFunction|AbilityPlayerDistancetoVector (.ImpactPoint (Collision|BreakHitResult _outhit))) (Variables|Const|GetMaxDistanceToGroundBeforePredictionComplete)) (> (CallFunction|AbilityPlayerFallSpeed) (Variables|Const|GetMinFallSpeedToCheckPredictionDistance))) (CallFunction|FloatWithinRange (/ (/ (CallFunction|AbilityPlayerDistancetoVector (.ImpactPoint (Collision|BreakHitResult _outhit))) (CallFunction|GetMovementAmountPerFrame)) (CallFunction|GetApproximateFramesPerSecond)) _min _max))
        (CallFunction|UnbindFallbackLandCheck)
        (Utilities|Time|ClearandInvalidateTimerbyHandle (Variables|Runtime|GetPredictLandingTimerHandle))
        (CallFunction|PredictionHit)
        (else
          (CallFunction|Predict)))
      (else
        (CallFunction|Predict)))
    (else
      (CallFunction|Predict))))

(event Custom|ListenForSwimmingAndSkydiving
  ;[1 1 1] Listen for water
  (bind _asynctaskproxy (Ability|Tasks|WaitGameplayTagAdd "(TagName=\"MovementMode.SurfaceSwimming\")" (Ability|GetAvatarActorfromActorInfo))
    (:then
      (Variables|Runtime|SetSurfaceSwimmingAsyncTask _asynctaskproxy)
      ;[1 1 1] listen for parachuting
      (bind _asynctaskproxy_2 (Ability|Tasks|WaitGameplayTagAdd "(TagName=\"MovementMode.Skydiving\")" (Ability|GetAvatarActorfromActorInfo))
        (:then
          (Variables|Runtime|SetSkydivingAsyncTask _asynctaskproxy_2))
        (:Added
          (Variables|Default|SetbIgnoreLanded true)
          (Ability|CancelAbility))))
    (:Added
      (Variables|Default|SetbIgnoreLanded true)
      (Ability|CancelAbility))))

(event Custom|PredictionHit
  ;[1 1 1 1 64] Prediction Complete
  (Utilities|FlowControl|DoOnce false
    (:Completed
      (if (not (Variables|Default|GetHasLanded))
        (Ability|Tasks|PlayMontageAndWait "None" (Variables|AnimationSettings|GetPredictionToGroundMontage))))))

(event Custom|ResetPredictionHit
  ;[1 1 1 1 64] Prediction Complete
  (Utilities|FlowControl|DoOnce false
    (:Completed
      (if (not (Variables|Default|GetHasLanded))
        (Ability|Tasks|PlayMontageAndWait "None" (Variables|AnimationSettings|GetPredictionToGroundMontage))))))

(event Custom|WaitForLand
  ;[1 1 1 1 64] On Landed
  (Utilities|FlowControl|DoOnce false
    (:Completed
      (bind _asynctaskproxy (Ability|Tasks|WaitMovementModeChange "MOVE_Walking")
        (:then
          (Variables|Runtime|SetWaitForLandAsyncTask _asynctaskproxy)
          (EventDispatchers|AssignOnPawnLanded (Variables|Runtime|GetOut%20Fort%20Player%20Pawn) (delegate Custom|AbilityPlayerPawnLanded))
          (Ability|ExecuteGameplayCueOnOwner (Variables|Const|GetJumpAttackPredictLandTag) (Class|AthenaFunctionLibrary|MakeSimpleFortEffectContext (Ability|GetAvatarActorfromActorInfo) (Ability|GetAvatarActorfromActorInfo))))
        (:OnChange
          (CallFunction|UnbindAbilityPlayerPawnLanded)
          (CallFunction|PlayAbliityPlayerPawnLandedEvent))))))

(event Custom|ResetWaitForLand
  ;[1 1 1 1 64] On Landed
  (Utilities|FlowControl|DoOnce false
    (:Completed
      (bind _asynctaskproxy (Ability|Tasks|WaitMovementModeChange "MOVE_Walking")
        (:then
          (Variables|Runtime|SetWaitForLandAsyncTask _asynctaskproxy)
          (EventDispatchers|AssignOnPawnLanded (Variables|Runtime|GetOut%20Fort%20Player%20Pawn) (delegate Custom|AbilityPlayerPawnLanded))
          (Ability|ExecuteGameplayCueOnOwner (Variables|Const|GetJumpAttackPredictLandTag) (Class|AthenaFunctionLibrary|MakeSimpleFortEffectContext (Ability|GetAvatarActorfromActorInfo) (Ability|GetAvatarActorfromActorInfo))))
        (:OnChange
          (CallFunction|UnbindAbilityPlayerPawnLanded)
          (CallFunction|PlayAbliityPlayerPawnLandedEvent))))))

(event Custom|UnbindAbilityPlayerPawnLanded
  ;[1 1 1 1 64] On Landed
  (EventDispatchers|UnbindEventfromOnPawnLanded (Variables|Runtime|GetOut%20Fort%20Player%20Pawn) (delegate Custom|AbilityPlayerPawnLanded)))

(event Ability|EventActivateAbility
  ;[1 1 1 1 64] Ability
  (if (ScalableFloat|GetValueasBool (Variables|Default|GetJumpAttackEnabledHotfixRow))
    (bind _outfortplayerpawn (CallFunction|GetAvatarasValidFortPlayerPawn)
      (:ValidFortPlayerPawn
        (Variables|Runtime|SetOut%20Fort%20Player%20Pawn _outfortplayerpawn)
        (if (not (Variables|Default|GetBypassMeleeParentJumpEnabledCheck))
          (bind _ascreative_melee_parent (Utilities|Casting|CastToCreative_Melee_Parent (Weapon|GetCurrentWeapon (Variables|Runtime|GetOut%20Fort%20Player%20Pawn))))
          (Variables|Default|SetWeapon _ascreative_melee_parent)
          (bind _enabled (CreativeMeleeWeaponInterface|GetJumpAttackEnabled (Variables|Default|GetWeapon) :!class /CRD_Melee/Creative_Melee_Weapon_Interface.Creative_Melee_Weapon_Interface_C))
          (if (not _enabled)
            (Ability|EndAbility)))
        ;;[0.3118 0.5472 1 0.25] Only Jump attack if we have the energy for it.
        (bind (_primary_attack_cost _secondary_attack_cost _block_hit_cost _dodge_cost _sprint_cost _jump_attack_cost) (CreativeMeleeWeaponInterface|GetEnergyCosts(Message) (Variables|Default|GetWeapon) false))
        (if (> _jump_attack_cost 0)
          (bind (_outamountofenergyused _returnvalue) (Energy|UseEnergy (Actor|GetComponentbyClass (Variables|Runtime|GetOut%20Fort%20Player%20Pawn) "/TacticalSprintGame/Gameplay/TacticalSprintEnergyComponent.TacticalSprintEnergyComponent_C") _jump_attack_cost false 0))
          (if (not _returnvalue)
            (FailedAction|UnabletoPerformAction (Variables|Runtime|GetOut%20Fort%20Player%20Pawn) true (Variables|Default|GetFailed%20Reason))
            (GameplayCue|ExecuteGameplayCue (Variables|Runtime|GetOut%20Fort%20Player%20Pawn) (Variables|Default|GetGCN_NotEnoughEnergy) (Class|AthenaFunctionLibrary|MakeSimpleFortEffectContext (Variables|Runtime|GetOut%20Fort%20Player%20Pawn)))
            (Ability|CancelAbility))
          (else
            _))
        (Ability|SendGameplayEvent "(TagName=\"Event.Athena.Balloons.RemoveAll\")")
        (CallFunction|UpdateDelayTimefromGravity)
        (|CallParentFunction)
        (CallFunction|ApplyJumpAttackTag)
        (Variables|Default|SetbCanPerformHit true)
        (Variables|Default|SetHasLanded false)
        (Variables|Default|SetbIgnoreLanded false)
        (Utilities|FlowControl|Sequence
          (:then_0
            (CallFunction|ListenforSwimmingandSkydiving)
            (CallFunction|CancelifClambering))
          (:then_1
            (CallFunction|ActivateLoopingWeaponFX)
            (CallFunction|Reset)
            (Utilities|FlowControl|DoOnce false
              (:Completed
                (if (Variables|Default|GetShouldUseAimAssist)
                  (bind (_asynctaskproxy _targetinghandle) (Class|AsyncActionPerformTargeting|PerformTargetingAsyncAction (Variables|Runtime|GetOut%20Fort%20Player%20Pawn) (Variables|Const|GetTP_AimAssist))
                    (:Targeted
                      (if (Utilities|Array|IsEmpty (TargetingSystem|TargetingResults|GetTargetingResults "/Script/TargetingSystem.Default__TargetingSubsystem" _targetinghandle))
                        (Variables|Default|SetImpulseDirection (Math|Vector|VectorClampSizeMax (Transformation|GetVelocity (Variables|Runtime|GetOut%20Fort%20Player%20Pawn)) (Variables|/Script/Engine.CharacterMovementComponent|GetMaxWalkSpeed (Variables|/Script/Engine.Character|GetCharacterMovement (Variables|Runtime|GetOut%20Fort%20Player%20Pawn)))))
                        (else
                          (Variables|Default|SetImpulseDirection (Utilities|Operators|Subtract (Transformation|GetActorLocation (.HitActor (Collision|BreakHitResult (Utilities|Array|Get(aref) (TargetingSystem|TargetingResults|GetTargetingResults "/Script/TargetingSystem.Default__TargetingSubsystem" _targetinghandle))))) (Transformation|GetActorLocation (Variables|Runtime|GetOut%20Fort%20Player%20Pawn)) (Transformation|GetVelocity (Variables|Runtime|GetOut%20Fort%20Player%20Pawn))))))
                      (Pawn|Components|CharacterMovement|AddImpulse (Variables|/Script/Engine.Character|GetCharacterMovement (Variables|Runtime|GetOut%20Fort%20Player%20Pawn)) (Math|Vector|SelectVector (Variables|Default|GetOverrideImpulseDirection) (Math|Vector|MakeVector (.x (Variables|Default|GetImpulseDirection)) (.y (Variables|Default|GetImpulseDirection)) 3.6e+02) (Variables|Default|GetShouldOverrideImpulseDirection)) true)
                      (Ability|Tasks|PlayMontageAndWait "None" (Variables|AnimationSettings|GetJumpAndFallLoopMontage))
                      (CallFunction|Predict)))
                  (else
                    _)))))))
      (:AvatarCastFailed
        _))
    (else
      _)))

(event Custom|ResetActivation
  ;[1 1 1 1 64] Ability
  (Utilities|FlowControl|DoOnce false
    (:Completed
      (if (Variables|Default|GetShouldUseAimAssist)
        (bind (_asynctaskproxy _targetinghandle) (Class|AsyncActionPerformTargeting|PerformTargetingAsyncAction (Variables|Runtime|GetOut%20Fort%20Player%20Pawn) (Variables|Const|GetTP_AimAssist))
          (:Targeted
            (if (Utilities|Array|IsEmpty (TargetingSystem|TargetingResults|GetTargetingResults "/Script/TargetingSystem.Default__TargetingSubsystem" _targetinghandle))
              (Variables|Default|SetImpulseDirection (Math|Vector|VectorClampSizeMax (Transformation|GetVelocity (Variables|Runtime|GetOut%20Fort%20Player%20Pawn)) (Variables|/Script/Engine.CharacterMovementComponent|GetMaxWalkSpeed (Variables|/Script/Engine.Character|GetCharacterMovement (Variables|Runtime|GetOut%20Fort%20Player%20Pawn)))))
              (else
                (Variables|Default|SetImpulseDirection (Utilities|Operators|Subtract (Transformation|GetActorLocation (.HitActor (Collision|BreakHitResult (Utilities|Array|Get(aref) (TargetingSystem|TargetingResults|GetTargetingResults "/Script/TargetingSystem.Default__TargetingSubsystem" _targetinghandle))))) (Transformation|GetActorLocation (Variables|Runtime|GetOut%20Fort%20Player%20Pawn)) (Transformation|GetVelocity (Variables|Runtime|GetOut%20Fort%20Player%20Pawn))))))
            (Pawn|Components|CharacterMovement|AddImpulse (Variables|/Script/Engine.Character|GetCharacterMovement (Variables|Runtime|GetOut%20Fort%20Player%20Pawn)) (Math|Vector|SelectVector (Variables|Default|GetOverrideImpulseDirection) (Math|Vector|MakeVector (.x (Variables|Default|GetImpulseDirection)) (.y (Variables|Default|GetImpulseDirection)) 3.6e+02) (Variables|Default|GetShouldOverrideImpulseDirection)) true)
            (Ability|Tasks|PlayMontageAndWait "None" (Variables|AnimationSettings|GetJumpAndFallLoopMontage))
            (CallFunction|Predict)))
        (else
          _)))))

(event Custom|AbilityPlayerPawnLanded
  (params (Hit :type Hit%20Result%20Structure%20%28by%20ref%29))
  ;[1 1 1 1 64] On Landed
  (GameplayTasks|EndTask (Variables|Runtime|GetWaitForLandAsyncTask))
  (CallFunction|UnbindAbilityPlayerPawnLanded)
  (CallFunction|PlayAbliityPlayerPawnLandedEvent))

(event Custom|ForceLanding
  ;[1 1 1 1 64] On Landed
  (CallFunction|UnbindAbilityPlayerPawnLanded)
  (CallFunction|PlayAbliityPlayerPawnLandedEvent))

(event Custom|PlayAbliityPlayerPawnLandedEvent
  ;[1 1 1 1 64] On Landed
  (Utilities|FlowControl|DoOnce false
    (:Completed
      (CallFunction|DeactivateLoopingWeaponFX)
      (bind (_asynctaskproxy _notifytag) (Ability|Tasks|PlayMontageAndWaitNotify "None" (Variables|AnimationSettings|GetLandedOnGroundMontage) 1.0 "None" false)
        (:then
          (Variables|Default|SetHasLanded true)
          ;;[0 1 0.1364] This finishing on the client first can cause potential issues if the player swings the weapon before the server says its ok
          (CallFunction|DoTargetting)
          (CallFunction|OnPawnLanded))
        (:OnCompleted
          _)
        (:OnInterrupted
          _)
        (:OnCancelled
          _)
        (:OnNotifyBegin
          (if (== _notifytag (Variables|Const|GetRemoveAttackSuppressionTag))
            (CallFunction|RemovePrimaryAttackSuppression))))
      (Ability|Tasks|WaitNetSync "BothWait"
        (:OnSync
          ;[1 1 1] We want to wait a small amount of time because this is called _after_ pawn landed event but we need that extra time for checking tag and therfore adding knockback from the jump attack ability
          ;;[0.5156 0 0.0185]
          (Utilities|FlowControl|Delay 0.2)
          (CallFunction|RemoveJumpAttackTag)
          (Ability|EndAbility))))))

(event Custom|ResetPlayAblityPlayerPawnLandedEvent
  ;[1 1 1 1 64] On Landed
  (Utilities|FlowControl|DoOnce false
    (:Completed
      (CallFunction|DeactivateLoopingWeaponFX)
      (bind (_asynctaskproxy _notifytag) (Ability|Tasks|PlayMontageAndWaitNotify "None" (Variables|AnimationSettings|GetLandedOnGroundMontage) 1.0 "None" false)
        (:then
          (Variables|Default|SetHasLanded true)
          ;;[0 1 0.1364] This finishing on the client first can cause potential issues if the player swings the weapon before the server says its ok
          (CallFunction|DoTargetting)
          (CallFunction|OnPawnLanded))
        (:OnCompleted
          _)
        (:OnInterrupted
          _)
        (:OnCancelled
          _)
        (:OnNotifyBegin
          (if (== _notifytag (Variables|Const|GetRemoveAttackSuppressionTag))
            (CallFunction|RemovePrimaryAttackSuppression))))
      (Ability|Tasks|WaitNetSync "BothWait"
        (:OnSync
          ;[1 1 1] We want to wait a small amount of time because this is called _after_ pawn landed event but we need that extra time for checking tag and therfore adding knockback from the jump attack ability
          ;;[0.5156 0 0.0185]
          (Utilities|FlowControl|Delay 0.2)
          (CallFunction|RemoveJumpAttackTag)
          (Ability|EndAbility))))))

(event Custom|ForceLandedEvent
  ;[1 1 1 1 64] On Landed
  (CallFunction|DeactivateLoopingWeaponFX)
  (bind (_asynctaskproxy _notifytag) (Ability|Tasks|PlayMontageAndWaitNotify "None" (Variables|AnimationSettings|GetLandedOnGroundMontage) 1.0 "None" false)
    (:then
      (Variables|Default|SetHasLanded true)
      ;;[0 1 0.1364] This finishing on the client first can cause potential issues if the player swings the weapon before the server says its ok
      (CallFunction|DoTargetting)
      (CallFunction|OnPawnLanded))
    (:OnCompleted
      _)
    (:OnInterrupted
      _)
    (:OnCancelled
      _)
    (:OnNotifyBegin
      (if (== _notifytag (Variables|Const|GetRemoveAttackSuppressionTag))
        (CallFunction|RemovePrimaryAttackSuppression))))
  (Ability|Tasks|WaitNetSync "BothWait"
    (:OnSync
      ;[1 1 1] We want to wait a small amount of time because this is called _after_ pawn landed event but we need that extra time for checking tag and therfore adding knockback from the jump attack ability
      ;;[0.5156 0 0.0185]
      (Utilities|FlowControl|Delay 0.2)
      (CallFunction|RemoveJumpAttackTag)
      (Ability|EndAbility))))

(event Ability|EventOnEndAbility
  (params (bWasCancelled :type Boolean))
  ; Ability Complete
  (Variables|Default|SetbCanPerformHit false)
  (if bWasCancelled
    ;;[1 0.006821 0] This ablity was cancelled from water, finish up.
    (CallFunction|UnbindFallbackLandCheck)
    (CallFunction|ForceLanding)
    (CallFunction|ForceLandedEvent))
  (|CallParentFunction bWasCancelled)
  (Utilities|Time|ClearandInvalidateTimerbyHandle (Variables|Runtime|GetPredictLandingTimerHandle))
  (GameplayTasks|EndTask (Variables|Runtime|GetWaitForLandAsyncTask))
  (CallFunction|UnbindAbilityPlayerPawnLanded))

(event Custom|ActivateLoopingWeaponFX
  (bind _returnvalue (Ability|GetAvatarActorfromActorInfo))
  ;[1 1 1] Cues
  (GameplayCue|AddGameplayCue _returnvalue (Variables|Const|GetGCL_WeaponLoopingFX) (Class|AthenaFunctionLibrary|MakeSimpleFortEffectContext _returnvalue))
  (GameplayCue|AddGameplayCue _returnvalue (Variables|Const|GetGCL_JumpAttackFall)))

(event Custom|ApplyJumpAttackTag
  ; Jump Attack Tag Apply & Remove
  (bind _returnvalue (Ability|ApplyGameplayEffectSpecToOwner (Ability|MakeOutgoingGameplayEffectSpec (Variables|Default|GetGE_JumpAttackApplied))))
  (Variables|Default|SetJumpAttackAppliedHandle _returnvalue))

(event Custom|DeactivateLoopingWeaponFX
  (bind _returnvalue (Ability|GetAvatarActorfromActorInfo))
  ;[1 1 1] Cues
  (GameplayCue|RemoveGameplayCue _returnvalue (Variables|Const|GetGCL_WeaponLoopingFX))
  (GameplayCue|RemoveGameplayCue _returnvalue (Variables|Const|GetGCL_JumpAttackFall)))

(event Custom|RemoveJumpAttackTag
  ; Jump Attack Tag Apply & Remove
  (Ability|RemoveGameplayEffectFromOwnerWithHandle (Variables|Default|GetJumpAttackAppliedHandle)))

(event Custom|OnPawnLanded
  ;[1 1 1] Cues
  (Utilities|FlowControl|DoOnce false
    (:Completed
      (if (not (Variables|Default|GetbIgnoreLanded))
        (Ability|ExecuteGameplayCueOnOwner (Variables|Const|GetGCN_JumpAttackSlam) (Class|AthenaFunctionLibrary|MakeSimpleFortEffectContext (Ability|GetAvatarActorfromActorInfo) (Ability|GetAvatarActorfromActorInfo)))))))

(event Custom|DoTargetting
  (bind (_returnvalue_x _returnvalue_y) (CallFunction|GetNormalizedLaunchDirectionXY))
  (bind _knockup_multiplier (Variables|Default|GetKnockup%20Multiplier))
  (bind _power (Variables|Default|GetPower))
  (bind _returnvalue (* _power _knockup_multiplier))
  (bind _power_2 (Variables|Default|GetPower))
  (bind _returnvalue_2 (* _returnvalue_x _power_2))
  (bind _power_3 (Variables|Default|GetPower))
  (bind _returnvalue_3 (* _returnvalue_y _power_3))
  ;[1 1 1 1 64] Targeting
  (Utilities|FlowControl|DoOnce false
    (:Completed
      (if (Variables|Default|GetbCanPerformHit)
        (bind _outfortplayerpawn (CallFunction|GetAvatarasValidFortPlayerPawn)
          (:ValidFortPlayerPawn
            (bind _output_get (Variables|Default|SetFortPlayerpawn_0 _outfortplayerpawn))
            (CallFunction|ApplySecondaryFeedback)
            (bind (_asynctaskproxy _targetinghandle) (Class|AsyncActionPerformTargeting|PerformTargetingAsyncAction _output_get (Variables|Const|GetTargeting%20Preset))
              (:Targeted
                (for _array_element (TargetingSystem|TargetingResults|GetTargetingResults "/Script/TargetingSystem.Default__TargetingSubsystem" _targetinghandle)
                  (bind _output_get_2 (Variables|Default|SetTargetData (Ability|TargetData|AbilityTargetDatafromHitResult _array_element)))
                  (Variables|Default|SetHit%20Actor (.HitActor (Collision|BreakHitResult (Ability|EffectContext|GetHitResult (Ability|GetContextfromOwner _output_get_2)))))
                  (CallFunction|ApplyEffectstoTargets _output_get_2 true)
                  (bind _ascreative_melee_parent (Utilities|Casting|CastToCreative_Melee_Parent (Variables|/Script/FortniteGame.FortPawn|GetCurrentWeapon (Variables|Default|GetFortPlayerpawn_0))))
                  (bind _output_get_3 (Variables|Default|SetMeleeParent _ascreative_melee_parent))
                  (bind _power_4 (CreativeMeleeWeaponInterface|GetJumpAttackKnockbackPower _output_get_3 :!class /CRD_Melee/Weapons/Creative_Melee_Parent.Creative_Melee_Parent_C))
                  (if (> _power_4 0)
                    (Variables|Default|SetKnockbackPower _power_4)
                    (Utilities|FlowControl|Sequence
                      (:then_0
                        (Utilities|Casting|CastToFortPlayerPawn (Variables|Default|GetHit%20Actor)
                          (:CastFailed
                            (bind _asfort_aipawn (Utilities|Casting|CastToFortAIPawn (Variables|Default|GetHit%20Actor))
                              (:then
                                (Utilities|IsValid (Variables|/Script/Engine.Character|GetCharacterMovement _asfort_aipawn)
                                  (:Is%20Valid
                                    ;;[1 0 0] Do we need to set it again here?
                                    (bind _power_5 (CreativeMeleeWeaponInterface|GetJumpAttackKnockbackPower (Variables|Default|GetMeleeParent) :!class /CRD_Melee/Creative_Melee_Weapon_Interface.Creative_Melee_Weapon_Interface_C))
                                    (Variables|Default|SetPower _power_5)
                                    (Pawn|Components|CharacterMovement|AddImpulse (Variables|/Script/Engine.Character|GetCharacterMovement _asfort_aipawn) (Math|Vector|MakeVector _returnvalue_2 _returnvalue_3 _returnvalue) true))))
                              (:CastFailed
                                (Ability|ApplyGameplayEffectToTarget (Variables|Default|GetTargetData) "/CRD_Melee/Weapons/GA/GE_Creative_Melee_Knockback_Status.GE_Creative_Melee_Knockback_Status_C"))))))
                      (:then_1
                        (bind _asfort_player_pawn (Utilities|Casting|CastToFortPlayerPawn (Variables|Default|GetHit%20Actor)))
                        (Utilities|IsValid (Variables|/Script/Engine.Character|GetCharacterMovement _asfort_player_pawn)
                          (:Is%20Valid
                            (bind _power_6 (CreativeMeleeWeaponInterface|GetJumpAttackKnockbackPower (Variables|Default|GetMeleeParent) :!class /CRD_Melee/Creative_Melee_Weapon_Interface.Creative_Melee_Weapon_Interface_C))
                            (Variables|Default|SetPower _power_6)
                            (Pawn|Components|CharacterMovement|AddImpulse (Variables|/Script/Engine.Character|GetCharacterMovement _asfort_player_pawn) (Math|Vector|MakeVector _returnvalue_2 _returnvalue_3 _returnvalue) true)))))))))))))))

(event Custom|ResetOnPawnLanded
  ;[1 1 1] Cues
  (Utilities|FlowControl|DoOnce false
    (:Completed
      (if (not (Variables|Default|GetbIgnoreLanded))
        (Ability|ExecuteGameplayCueOnOwner (Variables|Const|GetGCN_JumpAttackSlam) (Class|AthenaFunctionLibrary|MakeSimpleFortEffectContext (Ability|GetAvatarActorfromActorInfo) (Ability|GetAvatarActorfromActorInfo)))))))

(event Custom|ResetDoTargetting
  (bind (_returnvalue_x _returnvalue_y) (CallFunction|GetNormalizedLaunchDirectionXY))
  (bind _knockup_multiplier (Variables|Default|GetKnockup%20Multiplier))
  (bind _power (Variables|Default|GetPower))
  (bind _returnvalue (* _power _knockup_multiplier))
  (bind _power_2 (Variables|Default|GetPower))
  (bind _returnvalue_2 (* _returnvalue_x _power_2))
  (bind _power_3 (Variables|Default|GetPower))
  (bind _returnvalue_3 (* _returnvalue_y _power_3))
  ;[1 1 1 1 64] Targeting
  (Utilities|FlowControl|DoOnce false
    (:Completed
      (if (Variables|Default|GetbCanPerformHit)
        (bind _outfortplayerpawn (CallFunction|GetAvatarasValidFortPlayerPawn)
          (:ValidFortPlayerPawn
            (bind _output_get (Variables|Default|SetFortPlayerpawn_0 _outfortplayerpawn))
            (CallFunction|ApplySecondaryFeedback)
            (bind (_asynctaskproxy _targetinghandle) (Class|AsyncActionPerformTargeting|PerformTargetingAsyncAction _output_get (Variables|Const|GetTargeting%20Preset))
              (:Targeted
                (for _array_element (TargetingSystem|TargetingResults|GetTargetingResults "/Script/TargetingSystem.Default__TargetingSubsystem" _targetinghandle)
                  (bind _output_get_2 (Variables|Default|SetTargetData (Ability|TargetData|AbilityTargetDatafromHitResult _array_element)))
                  (Variables|Default|SetHit%20Actor (.HitActor (Collision|BreakHitResult (Ability|EffectContext|GetHitResult (Ability|GetContextfromOwner _output_get_2)))))
                  (CallFunction|ApplyEffectstoTargets _output_get_2 true)
                  (bind _ascreative_melee_parent (Utilities|Casting|CastToCreative_Melee_Parent (Variables|/Script/FortniteGame.FortPawn|GetCurrentWeapon (Variables|Default|GetFortPlayerpawn_0))))
                  (bind _output_get_3 (Variables|Default|SetMeleeParent _ascreative_melee_parent))
                  (bind _power_4 (CreativeMeleeWeaponInterface|GetJumpAttackKnockbackPower _output_get_3 :!class /CRD_Melee/Weapons/Creative_Melee_Parent.Creative_Melee_Parent_C))
                  (if (> _power_4 0)
                    (Variables|Default|SetKnockbackPower _power_4)
                    (Utilities|FlowControl|Sequence
                      (:then_0
                        (Utilities|Casting|CastToFortPlayerPawn (Variables|Default|GetHit%20Actor)
                          (:CastFailed
                            (bind _asfort_aipawn (Utilities|Casting|CastToFortAIPawn (Variables|Default|GetHit%20Actor))
                              (:then
                                (Utilities|IsValid (Variables|/Script/Engine.Character|GetCharacterMovement _asfort_aipawn)
                                  (:Is%20Valid
                                    ;;[1 0 0] Do we need to set it again here?
                                    (bind _power_5 (CreativeMeleeWeaponInterface|GetJumpAttackKnockbackPower (Variables|Default|GetMeleeParent) :!class /CRD_Melee/Creative_Melee_Weapon_Interface.Creative_Melee_Weapon_Interface_C))
                                    (Variables|Default|SetPower _power_5)
                                    (Pawn|Components|CharacterMovement|AddImpulse (Variables|/Script/Engine.Character|GetCharacterMovement _asfort_aipawn) (Math|Vector|MakeVector _returnvalue_2 _returnvalue_3 _returnvalue) true))))
                              (:CastFailed
                                (Ability|ApplyGameplayEffectToTarget (Variables|Default|GetTargetData) "/CRD_Melee/Weapons/GA/GE_Creative_Melee_Knockback_Status.GE_Creative_Melee_Knockback_Status_C"))))))
                      (:then_1
                        (bind _asfort_player_pawn (Utilities|Casting|CastToFortPlayerPawn (Variables|Default|GetHit%20Actor)))
                        (Utilities|IsValid (Variables|/Script/Engine.Character|GetCharacterMovement _asfort_player_pawn)
                          (:Is%20Valid
                            (bind _power_6 (CreativeMeleeWeaponInterface|GetJumpAttackKnockbackPower (Variables|Default|GetMeleeParent) :!class /CRD_Melee/Creative_Melee_Weapon_Interface.Creative_Melee_Weapon_Interface_C))
                            (Variables|Default|SetPower _power_6)
                            (Pawn|Components|CharacterMovement|AddImpulse (Variables|/Script/Engine.Character|GetCharacterMovement _asfort_player_pawn) (Math|Vector|MakeVector _returnvalue_2 _returnvalue_3 _returnvalue) true)))))))))))))))

(fn AbilityPlayerFallSpeed
  (bind _out_fort_player_pawn (Variables|Runtime|GetOut%20Fort%20Player%20Pawn))
  (bind (_returnvalue_x _returnvalue_y _returnvalue_z) (Math|Vector|BreakVector (Transformation|GetVelocity _out_fort_player_pawn)))
  (if (< _returnvalue_z 0)
    (return (Math|Float|Absolute(Float) _returnvalue_z))
    (else
      (return 0.0))))

(fn AbilityPlayerDistanceToVector
  (params (ImpactPoint :type Vector))
  (return (Math|Vector|VectorLength (- (Transformation|GetActorLocation (Variables|Runtime|GetOut%20Fort%20Player%20Pawn)) ImpactPoint))))

(fn FloatWithinRange
  (params (Value :type Float%20%28double-precision%29) (Min :type Float%20%28double-precision%29) (Max :type Float%20%28double-precision%29))
  (return (and (>= Value Min) (<= Value Max))))

(fn GetNormalizedLaunchDirectionXY
  (bind (_returnvalue_x _returnvalue_y) (Math|Vector|BreakVector (Math|Vector|Normalize (- (Transformation|GetActorLocation (Variables|Default|GetHit%20Actor)) (Transformation|GetActorLocation (Ability|GetAvatarActorfromActorInfo))))))
  (return _returnvalue_x _returnvalue_y))

(fn GetApproximateFramesPerSecond
  (return (/ 1.0 (Utilities|Time|GetWorldDeltaSeconds))))

(fn GetMovementAmountPerFrame
  (return (* (Variables|/Script/Engine.CharacterMovementComponent|GetMaxWalkSpeed (Variables|/Script/Engine.Character|GetCharacterMovement (Variables|Runtime|GetOut%20Fort%20Player%20Pawn))) (Utilities|Time|GetWorldDeltaSeconds))))

(fn GetPlayerLocationWithCapsuleOffset
  (bind (_returnvalue_x _returnvalue_y _returnvalue_z) (Math|Vector|BreakVector (Transformation|GetActorLocation (Variables|Runtime|GetOut%20Fort%20Player%20Pawn))))
  (return _returnvalue_x _returnvalue_y (- _returnvalue_z (Components|Capsule|GetUnscaledCapsuleHalfHeight (Variables|/Script/Engine.Character|GetCapsuleComponent (Variables|Runtime|GetOut%20Fort%20Player%20Pawn))))))

(fn IsAbilityPlayerFalling
  (bind (_returnvalue_x _returnvalue_y _returnvalue_z) (Math|Vector|BreakVector (Transformation|GetVelocity (Variables|Runtime|GetOut%20Fort%20Player%20Pawn))))
  (return (< _returnvalue_z 0)))

(fn PredictionInRangeSeconds
  (return (- (Variables|Const|GetSecondsBeforeGroundToPlayPrediction) 0.1) (+ (Variables|Const|GetSecondsBeforeGroundToPlayPrediction) 0.1)))

(fn Reset
  (CallFunction|ResetPredictionHit)
  (CallFunction|ResetWaitforLand)
  (CallFunction|ResetPlayAblityPlayerPawnLandedEvent)
  (CallFunction|ResetActivation)
  (CallFunction|ResetDoTargetting)
  (CallFunction|ResetonPawnLanded))

(fn UpdateDelayTimeFromGravity
  (bind _outfortplayerpawn (CallFunction|GetAvatarasValidFortPlayerPawn)
    (:ValidFortPlayerPawn
      (Variables|Default|SetForce%20End%20Delay (+ (Variables|Default|GetBaseForceEndDuration) (/ (Variables|Default|GetDefaultGravityValue) (Components|Movement|GetGravityZ (Pawn|GetMovementComponent _outfortplayerpawn))))))))
