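; Parent gameplay ability for the Creative melee weapon dodge, gated by a hotfix value.
; On activation it caches pawn, weapon and TacticalSprint energy component, then spends the weapon's dodge energy cost.
; Insufficient energy cancels the ability with failure feedback and a gameplay cue.
; Dodging pushes a camera mode, applies an invincibility effect and calls parent activation; the camera pops when the ability ends uncancelled.
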(event Ability|EventActivateAbility
  (if (ScalableFloat|GetValueasBool (Variables|Default|GetDodgeEnabledHotfixRow))
    (CallFunction|InitAbility)
    ;[0.3118 0.5472 1 0.25] Only Dodge if we have the energy for it.
    (bind (_primary_attack_cost _secondary_attack_cost _block_hit_cost _dodge_cost) (CreativeMeleeWeaponInterface|GetEnergyCosts(Message) (Variables|Default|GetBP_WeaponActor) false))
    (if (> _dodge_cost 0)
      (bind (_outamountofenergyused _returnvalue) (Energy|UseEnergy (Variables|Default|GetEnergyComp) _dodge_cost false 0))
      (if (not _returnvalue)
        (FailedAction|UnabletoPerformAction (Variables|Default|GetPawn) true (Variables|Default|GetFailed%20Reason))
        (Ability|CancelAbility)
        (GameplayCue|ExecuteGameplayCue (Variables|Default|GetPawn) (Variables|Default|GetGCN_NotEnoughEnergy) (Class|AthenaFunctionLibrary|MakeSimpleFortEffectContext (Variables|Default|GetPawn))))
      (else
        _))
    (Camera|PushCameraMode (Variables|Default|GetCameraMode))
    (bind _outfortplayerpawn (CallFunction|GetAvatarasValidFortPlayerPawn)
      (:ValidFortPlayerPawn
        (Creative|GameplayEffects|SimpleApplyGameplayEffectToActor (Variables|Default|GetGE_DodgeInvincibility) _outfortplayerpawn))
      (:AvatarCastFailed
        _))
    (|CallParentFunction)
    (else
      (Ability|EndAbility))))

(event Ability|EventOnEndAbility
  (params (bWasCancelled :type Boolean))
  (if (not bWasCancelled)
    (Camera|PopCameraMode))
  (|CallParentFunction bWasCancelled))

(fn InitAbility
  ;[0.3118 0.5472 1 0.25] Get References to the Weapon and Energy components, we'll be needing them later.
  (Utilities|FlowControl|Sequence
    (:then_0
      (Variables|Default|GetBP_WeaponActor
        (:then)
        (:else
          (bind _outfortplayerpawn (CallFunction|GetAvatarasValidFortPlayerPawn)
            (:ValidFortPlayerPawn
              (Variables|Default|SetPawn _outfortplayerpawn)
              (Variables|Default|SetBP_WeaponActor (Weapon|GetCurrentWeapon _outfortplayerpawn)))))))
    (:then_1
      (Variables|Default|GetEnergyComp
        (:then)
        (:else
          (bind _pawn (Variables|Default|GetPawn)
            (:then
              (Variables|Default|SetEnergyComp (Actor|GetComponentbyClass _pawn "/TacticalSprintGame/Gameplay/TacticalSprintEnergyComponent.TacticalSprintEnergyComponent_C")))
            (:else)))))))
