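; Gameplay ability that supplies knockback for the secondary attack of Creative melee weapons, overriding launch hooks of a parent ability.
; FindLauncher resolves the source weapon from the launcher and reads secondary knockback and knockup through CreativeMeleeWeaponInterface.
; A jump-attack tag on the weapon's owner swaps in a fixed knockback power.
; GetInitialKnockMag scales normalized XY direction by knockback and sets Z to knockup, ignoring incoming Z and XY override.
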(event Ability|EventActivateAbility
  (|CallParentFunction) ; This node is disabled and will not be called.\nDrag off pins to build functionality.
)

(event Ability|EventOnEndAbility
  (params (bWasCancelled :type Boolean)))

(fn FindLauncher
  (params (DefaultLauncher :type Actor%20Object%20Reference))
  (Utilities|FlowControl|Sequence
    (:then_0
      (CallFunction|GenerateValuesFromLauncher DefaultLauncher))
    (:then_1
      (return DefaultLauncher))))

(fn GenerateValuesfromLauncher
  (params (Launcher%20Actor :type Actor%20Object%20Reference))
  (bind _secondarysourceweapon (Variables|Default|GetSecondarySourceWeapon))
  ;[1 1 1] Knockback from secondary ability values.
  (bind _asfort_pawn (Utilities|Casting|CastToFortPawn Launcher%20Actor)
    (:then
      (Variables|Default|SetSecondarySourceWeapon (Weapon|GetCurrentWeapon _asfort_pawn)))
    (:CastFailed
      (Variables|Default|SetSecondarySourceWeapon Launcher%20Actor)))
  (bind (_abilitytag _secondary_damage _secondary_knockback _secondary_knockup) (CreativeMeleeWeaponInterface|GetSecondaryAbility(Message) _secondarysourceweapon))
  (bind _asfort_weapon (Utilities|Casting|CastToFortWeapon _secondarysourceweapon)
    (:then
      (bind _returnvalue (GameplayTags|HasMatchingGameplayTag(Message) (Actor|GetOwner _asfort_weapon) (LocalVariables|GetJumpAttackGameplayTag)))
      (if _returnvalue
        (Variables|Default|SetKnockback%20Power 6e+02)
        (else
          (Variables|Default|SetKnockback%20Power _secondary_knockback)))
      (Variables|Default|SetKnockup%20Power _secondary_knockup))
    (:CastFailed
      _)))

(fn GetInitialKnockMag
  (params (Direction_X :type Float%20%28double-precision%29) (Direction_Y :type Float%20%28double-precision%29) (Direction_Z :type Float%20%28double-precision%29) (OverrideXYMag :type Boolean) (XYMAgOverride :type Float%20%28double-precision%29))
  (bind (_returnvalue_x _returnvalue_y) (Math|Vector2D|BreakVector2D (Math|Vector2D|Normalize2D (Math|Vector2D|MakeVector2D Direction_X Direction_Y))))
  (return (Math|Vector|MakeVector (* _returnvalue_x (Variables|Default|GetKnockback%20Power)) (* _returnvalue_y (Variables|Default|GetKnockback%20Power)) (Variables|Default|GetKnockup%20Power))))
