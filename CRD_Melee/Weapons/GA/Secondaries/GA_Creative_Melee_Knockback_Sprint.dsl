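; Gameplay ability that supplies the knockback for a creative melee weapon's sprint attack, overriding value functions of a parent knockback ability.
; When resolving the launcher, it takes the pawn's current weapon as source, or the launcher itself if the pawn cast fails.
; Knockback power comes from the weapon interface's sprint attack damage call, while knockup is a fixed constant.
; The launch vector scales the normalized XY direction by knockback and sets Z to knockup, ignoring the magnitude override inputs.
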(event Ability|EventActivateAbility
  (|CallParentFunction) ; This node is disabled and will not be called.\nDrag off pins to build functionality.
)

(event Ability|EventOnEndAbility
  (params (bWasCancelled :type Boolean)))

(fn FindLauncher
  (params (DefaultLauncher :type Actor%20Object%20Reference))
  (Utilities|FlowControl|Sequence
    (:then_0
      (CallFunction|GenerateValuesFromLauncher DefaultLauncher))
    (:then_1
      (return DefaultLauncher))))

(fn GenerateValuesfromLauncher
  (params (Launcher%20Actor :type Actor%20Object%20Reference))
  ;[1 1 1] Knockback from secondary ability values.
  (bind _asfort_pawn (Utilities|Casting|CastToFortPawn Launcher%20Actor)
    (:then
      (Variables|Default|SetSourceWeapon (Weapon|GetCurrentWeapon _asfort_pawn)))
    (:CastFailed
      (Variables|Default|SetSourceWeapon Launcher%20Actor)))
  (bind (_damage _knockback) (CreativeMeleeWeaponInterface|GetSprintAttackDamage(Message) (Variables|Default|GetSourceWeapon)))
  (Variables|Default|SetKnockback%20Power _knockback)
  (Variables|Default|SetKnockup%20Power 6e+02))

(fn GetInitialKnockMag
  (params (Direction_X :type Float%20%28double-precision%29) (Direction_Y :type Float%20%28double-precision%29) (Direction_Z :type Float%20%28double-precision%29) (OverrideXYMag :type Boolean) (XYMAgOverride :type Float%20%28double-precision%29))
  (bind (_returnvalue_x _returnvalue_y) (Math|Vector2D|BreakVector2D (Math|Vector2D|Normalize2D (Math|Vector2D|MakeVector2D Direction_X Direction_Y))))
  (return (Math|Vector|MakeVector (* _returnvalue_x (Variables|Default|GetKnockback%20Power)) (* _returnvalue_y (Variables|Default|GetKnockback%20Power)) (Variables|Default|GetKnockup%20Power))))
