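; Parent gameplay ability for Creative melee sprint attacks, which are gated by a hotfix value and the weapon's setting.
; Activation applies a SprintAttackAvailable effect and listens for tactical sprint ending; sprint end or ability end triggers a DoOnce-guarded attack montage.
; The montage notify spends sprint energy, runs async targeting, then applies weapon-driven SetByCaller damage, knockback, optional block break and hit cues.
; Ending the ability unbinds the sprint listener and removes the effect.
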(event Ability|EventActivateAbility
  (bind _fort_player_pawn (Variables|Default|GetFort%20Player%20Pawn))
  (bind _returnvalue (Pawn|GetController _fort_player_pawn))
  (bind _returnvalue_2 (Actor|GetComponentbyClass _returnvalue "/TacticalSprintGame/Gameplay/TacticalSprintControllerComponent.TacticalSprintControllerComponent_C"))
  ;[1 1 1] Activate
  (if (ScalableFloat|GetValueasBool (Variables|Default|GetSprintAttackEnabledHotfixRow))
    (bind _outfortplayerpawn (CallFunction|GetAvatarasValidFortPlayerPawn)
      (:ValidFortPlayerPawn
        (bind _output_get (Variables|Default|SetFort%20Player%20Pawn _outfortplayerpawn))
        (bind _ascreative_melee_parent (Utilities|Casting|CastToCreative_Melee_Parent (Weapon|GetCurrentWeapon _output_get))
          (:then
            (Variables|Default|SetWeapon _ascreative_melee_parent)
            (bind _enabled (CreativeMeleeWeaponInterface|GetSprintAttackEnabled (Variables|Default|GetWeapon) :!class /CRD_Melee/Weapons/Creative_Melee_Parent.Creative_Melee_Parent_C))
            (if _enabled
              (Utilities|IsValid _returnvalue_2
                (:Is%20Valid
                  (bind _output_get_2 (Variables|Default|SetTacticalSprintComponent _returnvalue_2))
                  (Sprint|AssignOnTacticalSprintEnded _output_get_2 (EventDispatchers|CreateEvent :!event Custom|OnTacticalSprintEnded_Event)) ; Listen for sprint ended
                  (Ability|ApplyGameplayEffectToOwner "/CRD_Melee/Weapons/GE/GE_SprintAttackAvailable.GE_SprintAttackAvailable_C"))
                (:Is%20Not%20Valid
                  _)) ; Does this pawn have tac sprint?
              (else
                _))
            (Ability|EndAbility))
          (:CastFailed
            _)))
      (:AvatarCastFailed
        _))
    (else
      (Ability|EndAbility))))

(event Custom|OnTacticalSprintEnded_Event
  (bind (_montagetoplay _animplayrate _animrootmotiontranslationscale) (Utilities|Struct|BreakFortGameplayAbilityMontageInfo (Variables|Animation|GetMontageInfo)))
  ;[1 1 1] Do Attack
  (Utilities|FlowControl|DoOnce false
    (:Completed
      (Ability|Tasks|PlayMontageAndWaitNotify "None" _montagetoplay _animplayrate "None" false _animrootmotiontranslationscale
        (:OnCompleted
          _)
        (:OnInterrupted
          _)
        (:OnCancelled
          _)
        (:OnNotifyBegin
          (bind _outfortplayerpawn (CallFunction|GetAvatarasValidFortPlayerPawn)
            (:ValidFortPlayerPawn
              (Utilities|IsValid (Variables|Default|GetTP_SprintAttack)
                (:Is%20Valid
                  ;;[0.3118 0.5472 1 0.25] Only Sprint attack if we have the energy for it.
                  (bind (_primary_attack_cost _secondary_attack_cost _block_hit_cost _dodge_cost _sprint_cost) (CreativeMeleeWeaponInterface|GetEnergyCosts(Message) (Variables|Default|GetWeapon) false))
                  (if (> _sprint_cost 0)
                    (bind (_outamountofenergyused _returnvalue) (Energy|UseEnergy (Actor|GetComponentbyClass _outfortplayerpawn "/TacticalSprintGame/Gameplay/TacticalSprintEnergyComponent.TacticalSprintEnergyComponent_C") _sprint_cost false 0))
                    (if (not _returnvalue)
                      (FailedAction|UnabletoPerformAction _outfortplayerpawn true (Variables|Default|GetFailed%20Reason))
                      (GameplayCue|ExecuteGameplayCue _outfortplayerpawn (Variables|Default|GetGCN_NotEnoughEnergy) (Class|AthenaFunctionLibrary|MakeSimpleFortEffectContext _outfortplayerpawn))
                      (Ability|CancelAbility))
                    (else
                      _))
                  (bind (_asynctaskproxy _targetinghandle) (Class|AsyncActionPerformTargeting|PerformTargetingAsyncAction _outfortplayerpawn (Variables|Default|GetTP_SprintAttack))
                    (:Targeted
                      (if (Utilities|Array|IsNotEmpty (TargetingSystem|TargetingResults|GetTargetingResults "/Script/TargetingSystem.Default__TargetingSubsystem" _targetinghandle))
                        (Ability|CommitAbility) ; Let's do this
                        (for _array_element (TargetingSystem|TargetingResults|GetTargetingResults "/Script/TargetingSystem.Default__TargetingSubsystem" _targetinghandle)
                          (CallFunction|ApplyEffectstoTarget (Ability|TargetData|AbilityTargetDatafromHitResult _array_element)) ; Do damage / knockback as appropriate
                          (CallFunction|ExecuteGameplayCuesonTarget (Ability|TargetData|AbilityTargetDatafromHitResult _array_element)) ; Play sound / vfx
                        )
                        (Utilities|FlowControl|Delay (Variables|Default|GetCancelDuration))
                        (else
                          _))
                      (Ability|EndAbility))) ; Target the attack
                )
                (:Is%20Not%20Valid
                  _)) ; Sprint attack targeting preset valid?
            )
            (:AvatarCastFailed
              _)))) ; Play attack anim
      (Ability|EndAbility))))

(event Ability|EventOnEndAbility
  (params (bWasCancelled :type Boolean))
  (bind (_montagetoplay _animplayrate _animrootmotiontranslationscale) (Utilities|Struct|BreakFortGameplayAbilityMontageInfo (Variables|Animation|GetMontageInfo)))
  (bind _tacticalsprintcomponent (Variables|Default|GetTacticalSprintComponent))
  ;[1 1 1] End Ability
  (Utilities|FlowControl|Sequence
    (:then_0
      ;[1 1 1] Do Attack
      (Utilities|FlowControl|DoOnce false
        (:Completed
          (Ability|Tasks|PlayMontageAndWaitNotify "None" _montagetoplay _animplayrate "None" false _animrootmotiontranslationscale
            (:OnCompleted
              _)
            (:OnInterrupted
              _)
            (:OnCancelled
              _)
            (:OnNotifyBegin
              (bind _outfortplayerpawn (CallFunction|GetAvatarasValidFortPlayerPawn)
                (:ValidFortPlayerPawn
                  (Utilities|IsValid (Variables|Default|GetTP_SprintAttack)
                    (:Is%20Valid
                      ;;[0.3118 0.5472 1 0.25] Only Sprint attack if we have the energy for it.
                      (bind (_primary_attack_cost _secondary_attack_cost _block_hit_cost _dodge_cost _sprint_cost) (CreativeMeleeWeaponInterface|GetEnergyCosts(Message) (Variables|Default|GetWeapon) false))
                      (if (> _sprint_cost 0)
                        (bind (_outamountofenergyused _returnvalue) (Energy|UseEnergy (Actor|GetComponentbyClass _outfortplayerpawn "/TacticalSprintGame/Gameplay/TacticalSprintEnergyComponent.TacticalSprintEnergyComponent_C") _sprint_cost false 0))
                        (if (not _returnvalue)
                          (FailedAction|UnabletoPerformAction _outfortplayerpawn true (Variables|Default|GetFailed%20Reason))
                          (GameplayCue|ExecuteGameplayCue _outfortplayerpawn (Variables|Default|GetGCN_NotEnoughEnergy) (Class|AthenaFunctionLibrary|MakeSimpleFortEffectContext _outfortplayerpawn))
                          (Ability|CancelAbility))
                        (else
                          _))
                      (bind (_asynctaskproxy _targetinghandle) (Class|AsyncActionPerformTargeting|PerformTargetingAsyncAction _outfortplayerpawn (Variables|Default|GetTP_SprintAttack))
                        (:Targeted
                          (if (Utilities|Array|IsNotEmpty (TargetingSystem|TargetingResults|GetTargetingResults "/Script/TargetingSystem.Default__TargetingSubsystem" _targetinghandle))
                            (Ability|CommitAbility) ; Let's do this
                            (for _array_element (TargetingSystem|TargetingResults|GetTargetingResults "/Script/TargetingSystem.Default__TargetingSubsystem" _targetinghandle)
                              (CallFunction|ApplyEffectstoTarget (Ability|TargetData|AbilityTargetDatafromHitResult _array_element)) ; Do damage / knockback as appropriate
                              (CallFunction|ExecuteGameplayCuesonTarget (Ability|TargetData|AbilityTargetDatafromHitResult _array_element)) ; Play sound / vfx
                            )
                            (Utilities|FlowControl|Delay (Variables|Default|GetCancelDuration))
                            (else
                              _))
                          (Ability|EndAbility))) ; Target the attack
                    )
                    (:Is%20Not%20Valid
                      _)) ; Sprint attack targeting preset valid?
                )
                (:AvatarCastFailed
                  _)))) ; Play attack anim
          (Ability|EndAbility))))
    (:then_1
      (Utilities|IsValid _tacticalsprintcomponent
        (:Is%20Valid
          (Sprint|UnbindEventfromOnTacticalSprintEnded _tacticalsprintcomponent (EventDispatchers|CreateEvent :!event Custom|OnTacticalSprintEnded_Event))
          (GameplayEffects|RemoveActiveGameplayEffectbySourceEffect (Ability|GetFortAbilitySystemComponent (Variables|Default|GetFort%20Player%20Pawn)) "/CRD_Melee/Weapons/GE/GE_SprintAttackAvailable.GE_SprintAttackAvailable_C" 0 -1))))))

(event Custom|ApplyBlockBreak
  (params (TargetData :type Gameplay%20Ability%20Target%20Data%20Handle%20Structure) (IgnoreSecondaryBlockBreakEnabledUserOption :type Boolean))
  ;[0.3118 0.5472 1 0.25] Apply break block
  (bind _outfortplayerpawn (CallFunction|GetAvatarasValidFortPlayerPawn)
    (:ValidFortPlayerPawn
      (bind _ascreative_melee_parent (Utilities|Casting|CastToCreative_Melee_Parent (Variables|/Script/FortniteGame.FortPawn|GetCurrentWeapon _outfortplayerpawn)))
      (bind _enabled (CreativeMeleeWeaponInterface|GetSecondaryBreakBlockEnabled _ascreative_melee_parent :!class /CRD_Melee/Weapons/Creative_Melee_Parent.Creative_Melee_Parent_C))
      (if (or _enabled IgnoreSecondaryBlockBreakEnabledUserOption)
        (for _array_element (Ability|TargetData|GetAllActorsfromTargetData TargetData)
          (bind _returnvalue (GameplayTags|HasMatchingGameplayTag(Message) _array_element "(TagName=\"Gameplay.Action.Player.Blocking\")")) ; Only break block if they're blocking.
          (if _returnvalue
            (Creative|GameplayEffects|SimpleApplyGameplayEffectToActor (Variables|Default|GetGE_BlockBreak) _array_element)))))))

(fn ApplyEffectsToTarget
  (params (TargetData :type Gameplay%20Ability%20Target%20Data%20Handle%20Structure))
  (bind _returnvalue (Ability|GetAvatarActorfromActorInfo))
  (Utilities|IsValid _returnvalue
    (:Is%20Valid
      (bind _asfort_player_pawn (Utilities|Casting|CastToFortPlayerPawn _returnvalue))
      (LocalVariables|SetFortPawnAvatar _asfort_player_pawn)
      (bind _ascreative_melee_parent (Utilities|Casting|CastToCreative_Melee_Parent (Weapon|GetCurrentWeapon (LocalVariables|GetFortPawnAvatar))))
      (bind _damage (CreativeMeleeWeaponInterface|GetSprintAttackDamage _ascreative_melee_parent :!class /CRD_Melee/Weapons/Creative_Melee_Parent.Creative_Melee_Parent_C))
      (LocalVariables|SetDamage _damage)
      (bind (_base_damage _environmental _aidamage _critchance _critmult) (CreativeMeleeWeaponInterface|GetHitData :self _ascreative_melee_parent :!class /CRD_Melee/Weapons/Creative_Melee_Parent.Creative_Melee_Parent_C))
      (LocalVariables|SetCritChance _critchance)
      (LocalVariables|SetCritMultiplier _critmult)
      (bind _returnvalue_2 (Ability|GameplayEffect|AssignTagSetbyCallerMagnitude (GameplayEffects|MakeOutgoingSpec (Ability|GetFortAbilitySystemComponent (LocalVariables|GetFortPawnAvatar)) (LocalVariables|GetGE_Damage) 0.0 (GameplayEffects|MakeFortEffectContext (Ability|GetFortAbilitySystemComponent (LocalVariables|GetFortPawnAvatar)) (Weapon|GetCurrentWeapon (LocalVariables|GetFortPawnAvatar)) 0 (LocalVariables|GetFortPawnAvatar) 0)) "(TagName=\"SetByCaller.MeleeManager.BaseDamage\")" (LocalVariables|GetDamage)))
      (bind _returnvalue_3 (Ability|GameplayEffect|AssignTagSetbyCallerMagnitude _returnvalue_2 "(TagName=\"SetByCaller.MeleeManager.EnvironmentDamage\")" (LocalVariables|GetDamage)))
      (bind _returnvalue_4 (Ability|GameplayEffect|AssignTagSetbyCallerMagnitude _returnvalue_3 "(TagName=\"SetByCaller.MeleeManager.CritChanceModifier\")" (LocalVariables|GetCritChance)))
      (bind _returnvalue_5 (Ability|GameplayEffect|AssignTagSetbyCallerMagnitude _returnvalue_4 "(TagName=\"SetByCaller.MeleeManager.CritModifier\")" (LocalVariables|GetCritMultiplier)))
      (Ability|ApplyGameplayEffectSpecToTarget _returnvalue_5 (Variables|GetTargetData))
      (Ability|ApplyGameplayEffectToTarget (Variables|GetTargetData) (LocalVariables|GetGE_Knockback) 0)
      (if (Variables|Default|GetblockBreak)
        (CallFunction|ApplyBlockBreak (Variables|GetTargetData))))))

(fn ExecuteGameplayCuesOnTarget
  (params (TargetData :type Gameplay%20Ability%20Target%20Data%20Handle%20Structure))
  (for _array_element (Ability|TargetData|GetAllActorsfromTargetData (Ability|TargetData|FilterTargetData (Variables|GetTargetData) (Filter|MakeFilterHandle (Utilities|Struct|MakeGameplayTargetDataFilter "/Script/FortniteGame.FortPawn" "TDFS_NoSelf") (Ability|GetAvatarActorfromActorInfo))))
    (Ability|ExecuteGameplayCueWithParamsOnOwner (Variables|Default|GetGCN_EnemyHit) (Ability|GameplayCue|MakeGameplayCueParameters :NormalizedMagnitude 0.0 :RawMagnitude 0.0 :Location (Transformation|GetActorLocation _array_element) :Normal "0,0,0" :GameplayEffectLevel 1 :AbilityLevel 1 :bReplicateLocationWhenUsingMinimalRepProxy false)))
  (for _array_element_2 (Ability|TargetData|GetAllActorsfromTargetData (Ability|TargetData|FilterTargetData (Variables|GetTargetData) (Filter|MakeFilterHandle (Utilities|Struct|MakeGameplayTargetDataFilter "/Script/FortniteGame.FortPawn" "TDFS_Any" true) (Ability|GetAvatarActorfromActorInfo))))
    (Ability|ExecuteGameplayCueWithParamsOnOwner (Variables|Default|GetGCN_EnvironmentHit) (Ability|GameplayCue|MakeGameplayCueParameters :NormalizedMagnitude 0.0 :RawMagnitude 0.0 :Location (Transformation|GetActorLocation _array_element_2) :Normal "0,0,0" :GameplayEffectLevel 1 :AbilityLevel 1 :bReplicateLocationWhenUsingMinimalRepProxy false))))
