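; Mantis melee secondary gameplay ability that performs a dash attack, gated by a hotfix toggle.
; Activation turns the pawn toward its movement input, or control-rotation forward when idle, then commits and fires a dash cue.
; The dash montage ignores pawn collision while continuous target selection applies effects on a hit, until a notify tag closes the window.
; When the montage completes or aborts, collision is restored and the ability ends.
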(event Custom|TriggerDashAnimationForWeapon
  (bind (_montageinfo_montagetoplay _montageinfo_animplayrate _montageinfo_animrootmotiontranslationscale _montageinfo_montagesectiontoplay _montageinfo_overridesection) (Utilities|Struct|BreakFortGameplayAbilityMontageInfo (Variables|GetMontageInfo_MontageToPlay)))
  ;[0 1 0.3342] The timing of this needs to be adjusted so we can get an event before "completed". as of now, there is an awkward delay before the player can swing again
  (bind (_asynctaskproxy _notifytag) (Ability|Tasks|PlayMontageAndWaitNotify "None" _montageinfo_montagetoplay _montageinfo_animplayrate _montageinfo_overridesection true _montageinfo_animrootmotiontranslationscale)
    (:then
      (CallFunction|SetPlayerCollisionAgainstOpposing false)
      (Variables|DashData|SetKeepTargeting true)
      (bind (_asynctaskproxy_2 _targetdata) (Ability|Tasks|DoTargetSelectionContinuous "None" "(TagName=\"Gameplay.Melee.Hit1\")")
        (:then
          _)
        (:Targeted
          (if (Variables|DashData|GetKeepTargeting)
            (CallFunction|ApplyEffectstoTargets _targetdata)
            (GameplayTasks|EndTask (Variables|DashData|GetAsync%20Attack))
            (Utilities|FlowControl|Delay 0.15))))
      (Variables|DashData|SetAsync%20Attack _asynctaskproxy_2))
    (:OnCompleted
      _)
    (:OnInterrupted
      _)
    (:OnCancelled
      _)
    (:OnNotifyBegin
      (if (== _notifytag (Variables|DashData|GetStopTargetingGameplayTag))
        (Variables|DashData|SetKeepTargeting false))))
  (CallFunction|SetPlayerCollisionAgainstOpposing true)
  (Ability|EndAbility))

(event Ability|EventActivateAbility
  (if (ScalableFloat|GetValueasBool (Variables|Default|GetDrillAttackEnabledHotfixRow))
    (|CallParentFunction)
    (CallFunction|ApplySecondaryFeedback)
    (Variables|DashData|SetAttackDirection (CallFunction|GetPawnMovementDirection))
    ;[1 1 1] This doesn't seem like the right method.
    (bind _outfortplayerpawn (CallFunction|GetAvatarasValidFortPlayerPawn)
      (:ValidFortPlayerPawn
        (Transformation|SetActorRotation _outfortplayerpawn (Math|Rotator|MakeRotator 0.0 0.0 (Math|Vector|GetYawPitchfromVector (Variables|DashData|GetAttackDirection))) false)
        (Ability|CommitAbility)
        (Ability|ExecuteGameplayCueOnOwner (Variables|DashData|GetGCN_Dash) (Class|AthenaFunctionLibrary|MakeSimpleFortEffectContext (Ability|GetAvatarActorfromActorInfo) (Ability|GetAvatarActorfromActorInfo)))
        (CallFunction|TriggerDashAnimationforWeapon))
      (:AvatarCastFailed
        (Ability|EndAbility)))
    (else
      (Ability|EndAbility))))

(event Ability|EventOnEndAbility
  (params (bWasCancelled :type Boolean))
  (|CallParentFunction bWasCancelled)
  (GameplayTasks|EndTask (Variables|DashData|GetAsyncTaskApplyRoot)))

(fn CanApplyDamageTo
  (params (In :type Array%20of%20Actor%20Object%20References))
  (for _array_element In
    (for _array_element_2 0
      (if (== _array_element_2 _array_element)
        (return false))))
  (return true))

(fn GetPawnMovementDirection
  (bind _outfortplayerpawn (CallFunction|GetAvatarasValidFortPlayerPawn)
    (:ValidFortPlayerPawn
      (bind (_returnvalue_roll _returnvalue_pitch _returnvalue_yaw) (Math|Rotator|BreakRotator (Pawn|GetControlRotation (Pawn|GetController _outfortplayerpawn) :!class /Script/Engine.Controller)))
      (return (Math|Vector|SelectVector (Math|Vector|GetForwardVector (Math|Rotator|MakeRotator _returnvalue_roll 0.0 _returnvalue_yaw)) (Math|Vector|Normalize (Pawn|Components|CharacterMovement|GetCurrentAcceleration (Variables|/Script/Engine.Character|GetCharacterMovement _outfortplayerpawn))) (Math|Vector|VectorIsNearlyZero (Pawn|Components|CharacterMovement|GetCurrentAcceleration (Variables|/Script/Engine.Character|GetCharacterMovement _outfortplayerpawn))))))))

(fn SetPlayerCollisionAgainstOpposing
  (params (bEnabled :type Boolean))
  (bind _outfortplayerpawn (CallFunction|GetAvatarasValidFortPlayerPawn)
    (:ValidFortPlayerPawn
      (if (Variables|GetbEnabled)
        (Collision|SetCollisionResponsetoChannel (Variables|/Script/Engine.Character|GetCapsuleComponent _outfortplayerpawn) "ECC_GameTraceChannel4" "ECR_Block")
        (Collision|SetCollisionResponsetoChannel (Variables|/Script/Engine.Character|GetCapsuleComponent _outfortplayerpawn) "ECC_Pawn" "ECR_Block")
        (else
          (Collision|SetCollisionResponsetoChannel (Variables|/Script/Engine.Character|GetCapsuleComponent _outfortplayerpawn) "ECC_GameTraceChannel4" "ECR_Ignore")
          (Collision|SetCollisionResponsetoChannel (Variables|/Script/Engine.Character|GetCapsuleComponent _outfortplayerpawn) "ECC_Pawn" "ECR_Ignore"))))
    (:AvatarCastFailed
      (Ability|EndAbility))))
