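; Mantis melee secondary heavy-attack ability: a hotfix-gated ground pound that launches the player upward with temporary air control.
; While falling, a timer-driven line trace estimates ground proximity from distance and speed to trigger a pre-land montage.
; Landing plays the land montage and cue, runs async targeting to apply effects to targets, then net-syncs before ending.
; Surface swimming or skydiving cancels it, and ending cleans up air control and the jump cue.
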(event Custom|SendPredictLandingEvent
  (Ability|Tasks|PlayMontageandWaitforTrigger "None" (Variables|Prediction|GetPrelandMontage)))

(event Custom|PredictLanding
  ;[1 1 1] Landing prediction
  (bind _returnvalue (Utilities|Time|SetTimerbyEvent (delegate Custom|PredictLand) (Variables|Prediction|GetPredictionLandingLoopTime)))
  (Variables|Prediction|SetPredictLandingHandle _returnvalue)
  (Utilities|FlowControl|DoOnce false
    (:Completed
      (Ability|ExecuteGameplayCueOnOwner (Variables|Prediction|GetHeavyAttackPredictland) (Class|AthenaFunctionLibrary|MakeSimpleFortEffectContext (Ability|GetAvatarActorfromActorInfo) (Ability|GetAvatarActorfromActorInfo))))))

(event Custom|PredictLand
  (bind (_returnvalue_x _returnvalue_y _returnvalue_z) (Math|Vector|BreakVector (Transformation|GetVelocity (Ability|GetAvatarActorfromActorInfo))))
  (bind (_min _max) (CallFunction|PredictionRangeinSeconds))
  (bind (_returnvalue_x_2 _returnvalue_y_2) (Math|Vector|BreakVector (* (Math|Vector|Normalize (Transformation|GetVelocity (Ability|GetAvatarActorfromActorInfo))) 1e+04)))
  ;[1 1 1] Landing prediction
  (if (< _returnvalue_z 0)
    (bind (_outhit _returnvalue) (Collision|LineTraceByChannel (Transformation|GetActorLocation (Ability|GetAvatarActorfromActorInfo)) (Utilities|Operators|Add (Transformation|GetActorLocation (Ability|GetAvatarActorfromActorInfo)) (Math|Vector|MakeVector _returnvalue_x_2 _returnvalue_y_2 (* 1e+04 -1.0))) "TraceTypeQuery10" false (Utilities|Array|MakeArray (Ability|GetAvatarActorfromActorInfo)) "None" true "(R=1.000000,G=0.000000,B=0.000000,A=1.000000)" "(R=0.000000,G=1.000000,B=0.000000,A=1.000000)" 1.0))
    (if _returnvalue
      (if (or (and (< (CallFunction|AbilityPlayerDistancetoVector (.ImpactPoint (Collision|BreakHitResult _outhit))) (Variables|Prediction|GetMaxDistanceToGroundBeforePredictionComplete)) (> (CallFunction|AbilityPlayerFallSpeed) (Variables|Prediction|GetMinFallSpeedToCheckPredictionDistance))) (CallFunction|FloatWithinRange (/ (/ (CallFunction|AbilityPlayerDistancetoVector (.ImpactPoint (Collision|BreakHitResult _outhit))) (CallFunction|GetMovementAmountPerFrame)) (CallFunction|GetApproximateFramesPerSecond)) _min _max))
        (Utilities|Time|ClearandInvalidateTimerbyHandle (Variables|Prediction|GetPredictLandingHandle))
        (CallFunction|SendPredictLandingEvent)
        (else
          (CallFunction|PredictLanding))))
    (else
      (CallFunction|PredictLanding))))

(event Custom|CancelIf
  (bind _asynctaskproxy (Ability|Tasks|WaitGameplayTagAdd "(TagName=\"MovementMode.SurfaceSwimming\")" (Ability|GetAvatarActorfromActorInfo))
    (:then
      (Variables|Default|SetSurfaceSwimmingAsyncTask _asynctaskproxy)
      (bind _asynctaskproxy_2 (Ability|Tasks|WaitGameplayTagAdd "(TagName=\"MovementMode.Skydiving\")" (Ability|GetAvatarActorfromActorInfo))
        (:then
          (Variables|Default|SetSkydivingAsyncTask _asynctaskproxy_2))
        (:Added
          (Ability|CancelAbility))))
    (:Added
      (Ability|CancelAbility))))

(event Custom|UnbindPawnLanded
  (bind _fortpawn (Variables|Runtime|GetFortPawn))
  (Utilities|IsValid _fortpawn
    (:Is%20Valid
      (EventDispatchers|UnbindEventfromOnPawnLanded _fortpawn (delegate Custom|PawnLanded))
      (Collision|AssignOnComponentHit (Variables|/Script/Engine.Character|GetCapsuleComponent (Variables|Runtime|GetFortPawn)) (delegate Custom|OnPawnCapsuleHit)))))

(event Custom|UnbindOnPlayerCapsuleHit
  (bind _fortpawn (Variables|Runtime|GetFortPawn))
  (bind _capsulecomponent (Variables|/Script/Engine.Character|GetCapsuleComponent _fortpawn))
  (Utilities|IsValid _fortpawn
    (:Is%20Valid
      (Utilities|IsValid _capsulecomponent
        (:Is%20Valid
          (Collision|UnbindEventfromOnComponentHit _capsulecomponent (delegate Custom|OnPawnCapsuleHit)))))))

(event Ability|EventActivateAbility
  (bind (_montageinfo_montagetoplay _montageinfo_animplayrate _montageinfo_animrootmotiontranslationscale _montageinfo_montagesectiontoplay _montageinfo_overridesection) (Utilities|Struct|BreakFortGameplayAbilityMontageInfo (Variables|GetMontageInfo_MontageToPlay)))
  (if (ScalableFloat|GetValueasBool (Variables|Default|GetGroundPoundEnabledHotfixRow))
    (|CallParentFunction)
    (CallFunction|CancelIf)
    (CallFunction|ApplySecondaryFeedback)
    (CallFunction|PawnLandedorHitReset)
    (bind _outfortplayerpawn (CallFunction|GetAvatarasValidFortPlayerPawn)
      (:ValidFortPlayerPawn
        (Variables|Runtime|SetFortPawn _outfortplayerpawn)
        (bind _returnvalue (Ability|ApplyGameplayEffectToOwner "/CRD_Melee/Weapons/GE/GE_CreativeMelee_AirControl.GE_CreativeMelee_AirControl_C"))
        (Variables|Const|SetAirControlGE _returnvalue)
        (CallFunction|ImpulsePlayer)
        (Ability|Tasks|PlayMontageandWaitforTrigger "None" _montageinfo_montagetoplay _montageinfo_animplayrate "FirstSection" _montageinfo_overridesection _montageinfo_animrootmotiontranslationscale)
        (GameplayCue|AddGameplayCue (Variables|Runtime|GetFortPawn) (Variables|Const|GetGCN_HeavyAttack_Jump) (Class|AthenaFunctionLibrary|MakeSimpleFortEffectContext (Variables|Runtime|GetFortPawn) (Variables|Runtime|GetFortPawn)))
        (CallFunction|PredictLanding)
        (EventDispatchers|AssignOnPawnLanded (Variables|Runtime|GetFortPawn) (delegate Custom|PawnLanded)))
      (:AvatarCastFailed
        (Ability|EndAbility)))
    (else
      (Ability|EndAbility))))

(event Custom|OnPawnCapsuleHit
  (params (HitComponent :type Primitive%20Component%20Object%20Reference) (OtherActor :type Actor%20Object%20Reference) (OtherComp :type Primitive%20Component%20Object%20Reference) (NormalImpulse :type Vector) (Hit :type Hit%20Result%20Structure%20%28by%20ref%29))
  (CallFunction|PawnLandedorHit)
  (CallFunction|UnbindonPlayerCapsuleHit))

(event Custom|PawnLanded
  (params (Hit :type Hit%20Result%20Structure%20%28by%20ref%29))
  (CallFunction|PawnLandedorHit)
  (CallFunction|UnbindPawnLanded))

(event Ability|EventOnEndAbility
  (params (bWasCancelled :type Boolean))
  (CallFunction|UnbindonPlayerCapsuleHit)
  (CallFunction|UnbindPawnLanded)
  (|CallParentFunction bWasCancelled)
  (GameplayTasks|EndTask (Variables|Default|GetSurfaceSwimmingAsyncTask))
  (CallFunction|RemoveAirMovementGE) ; Make sure air control gets removed\non cancel. (FORT-519713)
  (bind _returnvalue (Ability|GameplayCue|IsGameplayCueActive (Variables|Runtime|GetFortPawn) (Variables|Const|GetGCN_HeavyAttack_Jump))) ; Make sure jump cue gets removed\non cancel (FORT-519713)
  (if _returnvalue
    (GameplayCue|RemoveGameplayCue (Variables|Runtime|GetFortPawn) (Variables|Const|GetGCN_HeavyAttack_Jump))))

(event Custom|PawnLandedOrHit
  ;[1 1 1] Pawn landed on ground or hit something in the air.
  (Utilities|FlowControl|DoOnce false
    (:Completed
      (GameplayCue|RemoveGameplayCue (Variables|Runtime|GetFortPawn) (Variables|Const|GetGCN_HeavyAttack_Jump))
      (bind (_asynctaskproxy _notifytag) (Ability|Tasks|PlayMontageAndWaitNotify "None" (Variables|Animation|GetLeapLandMontage))
        (:then
          (CallFunction|RemoveAirMovementGE)
          (Ability|ExecuteGameplayCueOnOwner (Variables|Const|GetGCN_HeavyAttack_Land) (Class|AthenaFunctionLibrary|MakeSimpleFortEffectContext (Ability|GetAvatarActorfromActorInfo) (Ability|GetAvatarActorfromActorInfo)))
          (CallFunction|DoTargeting false)
          (CallFunction|ApplySecondaryFeedback))
        (:OnCompleted
          _)
        (:OnInterrupted
          _)
        (:OnCancelled
          _)
        (:OnNotifyBegin
          (if (== _notifytag (Variables|Default|GetRemove%20Attack%20Suppression%20Tag))
            (CallFunction|RemovePrimaryAttackSuppression))))
      (Ability|Tasks|WaitNetSync "BothWait"
        (:OnSync
          (Ability|EndAbility))))))

(event Custom|DoTargeting
  (params (%20JumpAttack :type Boolean))
  ;[1 1 1] Targeting
  (bind (_asynctaskproxy _targetinghandle) (Class|AsyncActionPerformTargeting|PerformTargetingAsyncAction (CallFunction|GetAvatarasFortPlayerPawn) (Variables|Targeting|GetTargeting%20Preset))
    (:Targeted
      (for _array_element (TargetingSystem|TargetingResults|GetTargetingResults "/Script/TargetingSystem.Default__TargetingSubsystem" _targetinghandle)
        (CallFunction|ApplyEffectstoTargets (Ability|TargetData|AbilityTargetDatafromHitResult _array_element))))))

(event Custom|PawnLandedOrHitReset
  ;[1 1 1] Pawn landed on ground or hit something in the air.
  (Utilities|FlowControl|DoOnce false
    (:Completed
      (GameplayCue|RemoveGameplayCue (Variables|Runtime|GetFortPawn) (Variables|Const|GetGCN_HeavyAttack_Jump))
      (bind (_asynctaskproxy _notifytag) (Ability|Tasks|PlayMontageAndWaitNotify "None" (Variables|Animation|GetLeapLandMontage))
        (:then
          (CallFunction|RemoveAirMovementGE)
          (Ability|ExecuteGameplayCueOnOwner (Variables|Const|GetGCN_HeavyAttack_Land) (Class|AthenaFunctionLibrary|MakeSimpleFortEffectContext (Ability|GetAvatarActorfromActorInfo) (Ability|GetAvatarActorfromActorInfo)))
          (CallFunction|DoTargeting false)
          (CallFunction|ApplySecondaryFeedback))
        (:OnCompleted
          _)
        (:OnInterrupted
          _)
        (:OnCancelled
          _)
        (:OnNotifyBegin
          (if (== _notifytag (Variables|Default|GetRemove%20Attack%20Suppression%20Tag))
            (CallFunction|RemovePrimaryAttackSuppression))))
      (Ability|Tasks|WaitNetSync "BothWait"
        (:OnSync
          (Ability|EndAbility))))))

(fn AbilityPlayerDistanceToVector
  (params (ImpactPoint :type Vector))
  (return (Math|Vector|VectorLength (- ImpactPoint (Transformation|GetActorLocation (Ability|GetAvatarActorfromActorInfo))))))

(fn AbilityPlayerFallSpeed
  (bind _fortpawn (Variables|Runtime|GetFortPawn))
  (bind (_returnvalue_x _returnvalue_y _returnvalue_z) (Math|Vector|BreakVector (Transformation|GetVelocity _fortpawn)))
  (if (< _returnvalue_z 0)
    (return (Math|Float|Absolute(Float) _returnvalue_z))
    (else
      (return 0.0))))

(fn FloatWithinRange
  (params (Value :type Float%20%28double-precision%29) (Min :type Float%20%28double-precision%29) (Max :type Float%20%28double-precision%29))
  (return (and (>= Value Min) (<= Value Max))))

(fn ForceEndMontages
  (bind _leaplandmontage (Variables|Animation|GetLeapLandMontage))
  (bind _prelandmontage (Variables|Prediction|GetPrelandMontage))
  (|CallParentFunction)
  (bind _outfortplayerpawn (CallFunction|GetAvatarasValidFortPlayerPawn)
    (:ValidFortPlayerPawn
      (Utilities|FlowControl|Sequence
        (:then_0
          (Utilities|IsValid _leaplandmontage
            (:Is%20Valid
              (Animation|StopAnimMontage _outfortplayerpawn _leaplandmontage))))
        (:then_1
          (Utilities|IsValid _prelandmontage
            (:Is%20Valid
              (Animation|StopAnimMontage _outfortplayerpawn _prelandmontage))))))))

(fn GetApproximateFramesPerSecond
  (return (/ 1.0 (Utilities|Time|GetWorldDeltaSeconds))))

(fn GetMovementAmountPerFrame
  (return (* (Variables|/Script/Engine.CharacterMovementComponent|GetMaxWalkSpeed (Variables|/Script/Engine.Character|GetCharacterMovement (Variables|Runtime|GetFortPawn))) (Utilities|Time|GetWorldDeltaSeconds))))

(fn ImpulsePlayer
  (bind (_returnvalue_x _returnvalue_y _returnvalue_z) (Math|Vector|BreakVector (Math|Vector|VectorClampSizeMax (Transformation|GetVelocity (Variables|Runtime|GetFortPawn)) (Variables|/Script/Engine.CharacterMovementComponent|GetMaxWalkSpeed (Variables|/Script/Engine.Character|GetCharacterMovement (Variables|Runtime|GetFortPawn))))))
  (Pawn|Character|LaunchCharacterJump (Variables|Runtime|GetFortPawn) (Math|Vector|MakeVector _returnvalue_x _returnvalue_y (Math|Float|Max(Float) (- (Variables|Const|GetLeapImpulseZ) (Math|Float|Max(Float) (.z (Transformation|GetVelocity (Variables|Runtime|GetFortPawn))))) (Variables|Const|GetMinLeapImpulse))) false (<= _returnvalue_z 0) true true))

(fn PredictionRangeInSeconds
  (return (+ (Variables|Prediction|GetSecondsBeforeGroundToPlayPrediction) 0.5) (- (Variables|Prediction|GetSecondsBeforeGroundToPlayPrediction) 0.5)))

(fn RemoveAirMovementGE
  (Ability|RemoveGameplayEffectFromOwnerWithHandle (Variables|Const|GetAirControlGE)))
