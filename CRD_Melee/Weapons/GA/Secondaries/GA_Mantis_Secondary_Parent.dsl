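; Parent gameplay ability for Creative melee (Mantis) weapon secondary attacks, providing shared activation, hit and feedback logic.
; Activation spends tactical sprint energy and suppresses primary attacks until end, which comes via cancel event, tactical sprint or safety timeout.
; Hit helpers check friendly fire, exempting player-built walls, and build SetByCaller damage specs from weapon data.
; Non-jump hits with positive damage add knockback, block break on blocking targets and hit cues.
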(event Custom|ApplySecondaryFeedback
  ;[1 1 1] Force feedback
  (bind _outfortplayerpawn (CallFunction|GetAvatarasValidFortPlayerPawn)
    (:ValidFortPlayerPawn
      (bind _ascreative_melee_parent (Utilities|Casting|CastToCreative_Melee_Parent (Variables|/Script/FortniteGame.FortPawn|GetCurrentWeapon _outfortplayerpawn)))
      (bind _scale (CreativeMeleeWeaponInterface|GetSecondaryFeedbackScale _ascreative_melee_parent :!class /CRD_Melee/Weapons/Creative_Melee_Parent.Creative_Melee_Parent_C))
      (switch Utilities|FlowControl|Switch|SwitchonCreative_Melee_Feedback_Scale _scale
        (:NewEnumerator1
          (CallFunction|PlayForceFeedbackandCameraShake (Variables|ForceFeedback|GetShake%20Class%20Light) (Variables|ForceFeedback|GetForce%20Feedback%20Effect%20Light)))
        (:NewEnumerator2
          (CallFunction|PlayForceFeedbackandCameraShake (Variables|ForceFeedback|GetShake%20Class%20Medium) (Variables|ForceFeedback|GetForce%20Feedback%20Effect%20Medium)))
        (:NewEnumerator3
          (CallFunction|PlayForceFeedbackandCameraShake (Variables|ForceFeedback|GetShake%20Class%20Heavy) (Variables|ForceFeedback|GetForce%20Feedback%20Effect%20Heavy)))))))

(event Custom|OnEffectsAppliedToTargets
  (params (TargetPawn :type Array%20of%20Actor%20Object%20References) (TargetEnvironment :type Array%20of%20Actor%20Object%20References))
  ;[1 1 1] Gameplay Cues for enemy and environment hits
  (for _array_element TargetPawn
    (Ability|ExecuteGameplayCueWithParamsOnOwner (Variables|Default|GetGCN_EnemyHit) (Ability|GameplayCue|MakeGameplayCueParameters :NormalizedMagnitude 0.0 :RawMagnitude 0.0 :Location (Transformation|GetActorLocation _array_element) :Normal "0,0,0" :GameplayEffectLevel 1 :AbilityLevel 1 :bReplicateLocationWhenUsingMinimalRepProxy false)))
  (for _array_element_2 TargetEnvironment
    (Ability|ExecuteGameplayCueWithParamsOnOwner (Variables|Default|GetGCN_EnvironmentHit) (Ability|GameplayCue|MakeGameplayCueParameters :NormalizedMagnitude 0.0 :RawMagnitude 0.0 :Location (Transformation|GetActorLocation _array_element_2) :Normal "0,0,0" :GameplayEffectLevel 1 :AbilityLevel 1 :bReplicateLocationWhenUsingMinimalRepProxy false))))

(event Custom|ApplyBlockBreak
  (params (TargetData :type Gameplay%20Ability%20Target%20Data%20Handle%20Structure) (IgnoreSecondaryBlockBreakEnabledUserOption :type Boolean))
  ;[0.3118 0.5472 1 0.25] Apply break block
  (bind _outfortplayerpawn (CallFunction|GetAvatarasValidFortPlayerPawn)
    (:ValidFortPlayerPawn
      (bind _ascreative_melee_parent (Utilities|Casting|CastToCreative_Melee_Parent (Variables|/Script/FortniteGame.FortPawn|GetCurrentWeapon _outfortplayerpawn)))
      (bind _enabled (CreativeMeleeWeaponInterface|GetSecondaryBreakBlockEnabled _ascreative_melee_parent :!class /CRD_Melee/Weapons/Creative_Melee_Parent.Creative_Melee_Parent_C))
      (if (or _enabled IgnoreSecondaryBlockBreakEnabledUserOption)
        (for _array_element (Ability|TargetData|GetAllActorsfromTargetData TargetData)
          (bind _returnvalue (GameplayTags|HasMatchingGameplayTag(Message) _array_element "(TagName=\"Gameplay.Action.Player.Blocking\")")) ; Only break block if they're blocking.
          (if _returnvalue
            (Creative|GameplayEffects|SimpleApplyGameplayEffectToActor (Variables|Default|GetGE_BlockBreak) _array_element)))))))

(event Ability|EventActivateAbility
  (bind (_isvalid _weapon) (CallFunction|GetWeapon))
  (if _isvalid
    (Variables|Default|SetWeapon _weapon)
    (CallFunction|UpdateEnergy)
    (Class|BMantisWeaponParent|ReticleStatusDisplay (Variables|Default|GetWeapon) true)
    (bind _returnvalue (Ability|ApplyGameplayEffectToOwner "/CRD_Melee/Weapons/GE/GE_BlockPrimaryAttack.GE_BlockPrimaryAttack_C"))
    (Variables|Default|SetBlockPrimaryEffectHandle _returnvalue)
    (Ability|Tasks|WaitGameplayEvent (Variables|Default|GetSecondary%20Cancel%20Tag)
      (:then
        ;[1 1 1] FORT-454406
        (Ability|Tasks|WaitGameplayTagAdd "(TagName=\"Gameplay.Action.Player.TacticalSprint\")" (Ability|GetAvatarActorfromActorInfo)
          (:then
            (Utilities|FlowControl|Delay (Variables|Default|GetForce%20End%20Delay)) ; Safety cancel
            (CallFunction|OnAbilityCancelled))
          (:Added
            (Ability|EndAbility))))
      (:EventReceived
        (CallFunction|OnAbilityCancelled)))
    (elif (Variables|Default|GetisOkayAsANonCreativeMeleeWeapon)
      _
      (else
        (Ability|EndAbility)))))

(event Custom|On%20Ability%20Cancelled
  (Ability|CancelAbility))

(event Ability|EventOnEndAbility
  (params (bWasCancelled :type Boolean))
  (CallFunction|RemovePrimaryAttackSuppression)
  (CallFunction|StopUsingEnergy)
  (CallFunction|ForceEndMontages)
  (bind _weapon (Variables|Default|GetWeapon)
    (:then
      (Class|BMantisWeaponParent|ReticleStatusDisplay _weapon false))
    (:else)))

(event Custom|Update%20Energy
  (CallFunction|StartUsingEnergy)
  (Utilities|FlowControl|Delay (Variables|Default|GetEnergy%20Use%20Duration))
  (CallFunction|StopUsingEnergy))

(fn ApplyEffectsToTargets
  (params (TargetData :type Gameplay%20Ability%20Target%20Data%20Handle%20Structure) (JumpAttack :type Boolean))
  ;[0 0.8002 1 0.25] Applies a Gameplay Effect to any targets using data from the source weapon to determine damage
  (bind _onsameteam (CallFunction|FriendlyFireCheck TargetData))
  (if (not _onsameteam)
    (Ability|ApplyGameplayEffectSpecToTarget (CallFunction|GetSecondaryDamageSpec (Variables|GetJumpAttack)) (Variables|GetTargetData))
    (if (not (Variables|GetJumpAttack))
      (bind (_abilitytag _secondary_damage) (CreativeMeleeWeaponInterface|GetSecondaryAbility(Message) (Variables|Default|GetWeapon)))
      ;[1 1 1] Knockback data is handled
      ; in the GA generated by the GE.
      ; All we care about right now is
      ; whether we want to or not.
      (if (> _secondary_damage 0.0)
        (Ability|ApplyGameplayEffectToTarget (Variables|GetTargetData) "/CRD_Melee/Weapons/GA/Secondaries/GE_Creative_Melee_Knockback_Secondary.GE_Creative_Melee_Knockback_Secondary_C")
        (CallFunction|ApplyBlockBreak (Variables|GetTargetData))
        (CallFunction|OnEffectsAppliedtoTargets (Ability|TargetData|GetAllActorsfromTargetData (Ability|TargetData|FilterTargetData (Variables|GetTargetData) (Filter|MakeFilterHandle (Utilities|Struct|MakeGameplayTargetDataFilter "/Script/FortniteGame.FortPawn" "TDFS_NoSelf") (Ability|GetAvatarActorfromActorInfo)))) (Ability|TargetData|GetAllActorsfromTargetData (Ability|TargetData|FilterTargetData (Variables|GetTargetData) (Filter|MakeFilterHandle (Utilities|Struct|MakeGameplayTargetDataFilter "/Script/FortniteGame.FortPawn" "TDFS_Any" true) (Ability|GetAvatarActorfromActorInfo)))))))))

(fn ForceEndMontages
  (bind _montageinfo_montagetoplay (Utilities|Struct|BreakFortGameplayAbilityMontageInfo (Variables|GetMontageInfo_MontageToPlay)))
  (bind _outfortplayerpawn (CallFunction|GetAvatarasValidFortPlayerPawn)
    (:ValidFortPlayerPawn
      (Animation|StopAnimMontage _outfortplayerpawn _montageinfo_montagetoplay))))

(fn FriendlyFireCheck
  (params (TargetData :type Gameplay%20Ability%20Target%20Data%20Handle%20Structure))
  (bind _returnvalue (Ability|GetContextfromOwner TargetData))
  (bind (_returnvalue_bblockinghit _returnvalue_binitialoverlap _returnvalue_time _returnvalue_distance _returnvalue_location _returnvalue_impactpoint _returnvalue_normal _returnvalue_impactnormal _returnvalue_physmat _returnvalue_hitactor) (Collision|BreakHitResult (Ability|EffectContext|GetHitResult _returnvalue)))
  (bind _returnvalue_2 (Ability|GetAbilitySystemComponent _returnvalue_hitactor))
  (bind _outfortplayerpawn (CallFunction|GetAvatarasValidFortPlayerPawn)
    (:ValidFortPlayerPawn
      (switch Utilities|FlowControl|Switch|SwitchonEFortTeamAffiliation (Team|GetTeamAffiliation _outfortplayerpawn _returnvalue_hitactor)
        (:Friendly
          (Utilities|IsValid (Minigame|GetMinigame _outfortplayerpawn :!class /Script/FortniteGame.MinigameBlueprintLibrary)
            (:Is%20Valid
              (if (Variables|/Script/FortniteGame.FortMinigameGameRulesComponent|GetbAllowFriendlyFire (Variables|/Script/FortniteGame.FortMinigame|GetGameRulesComponent (Minigame|GetMinigame _outfortplayerpawn :!class /Script/FortniteGame.MinigameBlueprintLibrary)))
                (return false)
                (else
                  ;[1 1 1] let player built walls be on a "different team" so they can take damage. this mimmics normal behavior of players being ablt to damage their own walls with other weapons but in this case, for secondary abilities
                  (Utilities|IsValid _returnvalue_2
                    (:Is%20Valid
                      (return (not (GameplayTags|HasMatchingGameplayTag _returnvalue_2 "(TagName=\"Building.Type.Player\")"))))
                    (:Is%20Not%20Valid
                      (return true))))))
            (:Is%20Not%20Valid
              (return true))))
        (:Neutral
          _)
        (:Hostile
          _))
      (return false))
    (:AvatarCastFailed
      (return false))))

(fn GetSecondaryDamageSpec
  (params (JumpAttack :type Boolean))
  (bind _fortplayerpawn (LocalVariables|GetFortPlayerPawn))
  (bind _currentweapon (Variables|/Script/FortniteGame.FortPawn|GetCurrentWeapon _fortplayerpawn))
  (bind _returnvalue (Ability|GetAvatarActorfromActorInfo))
  (bind _returnvalue_2 (Ability|GetFortAbilitySystemComponent _returnvalue))
  (bind _returnvalue_3 (GameplayEffects|MakeFortEffectContext _returnvalue_2 _currentweapon 0 _returnvalue 0))
  (bind _ge_class (Variables|Default|GetGE%20Class))
  (bind _returnvalue_4 (GameplayEffects|MakeOutgoingSpec _returnvalue_2 _ge_class 0.0 _returnvalue_3))
  (bind _damage (LocalVariables|GetDamage))
  (bind _asfort_player_pawn (Utilities|Casting|CastToFortPlayerPawn (Ability|GetAvatarActorfromActorInfo)))
  (LocalVariables|SetFortPlayerPawn _asfort_player_pawn)
  (bind _weapon (Variables|Default|GetWeapon)
    (:then
      (bind (_base_damage _environmental _aidamage _critchance _critmult) (CreativeMeleeWeaponInterface|GetHitData :self _weapon :!class /CRD_Melee/Creative_Melee_Weapon_Interface.Creative_Melee_Weapon_Interface_C))
      (LocalVariables|SetCritChance _critchance)
      (LocalVariables|SetCritMultiplier _critmult)
      (if JumpAttack
        (bind _damage_2 (CreativeMeleeWeaponInterface|GetJumpAttackDamage (Variables|Default|GetWeapon) :!class /CRD_Melee/Weapons/Creative_Melee_Parent.Creative_Melee_Parent_C))
        (LocalVariables|SetDamage _damage_2)
        (else
          (bind (_abilitytag _secondary_damage) (CreativeMeleeWeaponInterface|GetSecondaryAbility(Message) (Variables|Default|GetWeapon)))
          (LocalVariables|SetDamage _secondary_damage)))
      (bind _returnvalue_5 (Ability|GameplayEffect|AssignTagSetbyCallerMagnitude _returnvalue_4 "(TagName=\"SetByCaller.MeleeManager.BaseDamage\")" _damage))
      (bind _returnvalue_6 (Ability|GameplayEffect|AssignTagSetbyCallerMagnitude _returnvalue_5 "(TagName=\"SetByCaller.MeleeManager.EnvironmentDamage\")" _damage))
      (bind _returnvalue_7 (Ability|GameplayEffect|AssignTagSetbyCallerMagnitude _returnvalue_6 "(TagName=\"SetByCaller.MeleeManager.CritChanceModifier\")" (LocalVariables|GetCritChance)))
      (bind _returnvalue_8 (Ability|GameplayEffect|AssignTagSetbyCallerMagnitude _returnvalue_7 "(TagName=\"SetByCaller.MeleeManager.CritModifier\")" (LocalVariables|GetCritMultiplier)))
      (return _returnvalue_8))
    (:else
      (return _returnvalue_4))))

(fn GetMontageAsset
  (return (Utilities|Struct|BreakFortGameplayAbilityMontageInfo (Variables|Animation|GetMontageInfo))))

(fn GetWeapon
  (bind _outfortplayerpawn (CallFunction|GetAvatarasValidFortPlayerPawn)
    (:ValidFortPlayerPawn
      (bind _ascreative_melee_parent (Utilities|Casting|CastToCreative_Melee_Parent (Weapon|GetCurrentWeapon _outfortplayerpawn))
        (:then
          (return true _ascreative_melee_parent))
        (:CastFailed
          _)))
    (:AvatarCastFailed
      _))
  (return false 0))

(fn PlayForceFeedbackAndCameraShake
  (params (ShakeClass :type Legacy%20Camera%20Shake%20Class%20Reference) (ForceFeedbackEffect :type Force%20Feedback%20Effect%20Object%20Reference))
  (bind _outfortplayerpawn (CallFunction|GetAvatarasValidFortPlayerPawn)
    (:ValidFortPlayerPawn
      (CameraShakes|StartCameraShake :self (Variables|/Script/Engine.PlayerController|GetPlayerCameraManager (Fortnite|GetFortPlayerControllerfromActor _outfortplayerpawn)) :ShakeClass (Variables|GetShakeClass) :Scale 1.0 :PlaySpace "CameraLocal")
      (Fortnite|PlayLocalForceFeedbackatLocation (Variables|GetForceFeedbackEffect) (Transformation|GetActorLocation _outfortplayerpawn) 1024.0))))

(fn RemovePrimaryAttackSuppression
  (Ability|RemoveGameplayEffectFromOwnerWithHandle (Variables|Default|GetBlockPrimaryEffectHandle)))

(fn SetControlCharacter
  (params (Take%20Character%20Control :type Boolean))
  (bind _returnvalue (not Take%20Character%20Control))
  (bind _outfortplayerpawn (CallFunction|GetAvatarasValidFortPlayerPawn)
    (:ValidFortPlayerPawn
      (Input|SetIgnoreMoveInput (Pawn|GetController _outfortplayerpawn) Take%20Character%20Control)
      (Variables|/Script/Engine.CharacterMovementComponent|SetbOrientRotationToMovement Take%20Character%20Control (Variables|/Script/Engine.Character|GetCharacterMovement _outfortplayerpawn))
      (Variables|/Script/Engine.Pawn|SetbUseControllerRotationYaw _returnvalue _outfortplayerpawn)
      (Variables|/Script/Engine.CharacterMovementComponent|SetbUseControllerDesiredRotation _returnvalue (Variables|/Script/Engine.Character|GetCharacterMovement _outfortplayerpawn)))))

(fn StartUsingEnergy
  (bind _energycomponent (Variables|Default|GetEnergyComponent))
  (Variables|Default|GetEnergyComponent
    (:then
      _)
    (:else
      (bind _outfortplayerpawn (CallFunction|GetAvatarasValidFortPlayerPawn)
        (:ValidFortPlayerPawn
          (Variables|Default|SetEnergyComponent (Actor|GetComponentbyClass _outfortplayerpawn "/TacticalSprintGame/Gameplay/TacticalSprintEnergyComponent.TacticalSprintEnergyComponent_C"))))))
  (bind (_primary_attack_cost _secondary_attack_cost) (CreativeMeleeWeaponInterface|GetEnergyCosts (Variables|Default|GetWeapon) false :!class /CRD_Melee/Weapons/Creative_Melee_Parent.Creative_Melee_Parent_C))
  (if (== (Variables|Default|GetEnergy%20Use%20Duration) 0)
    (Energy|UseEnergy _energycomponent _secondary_attack_cost true 0)
    (else
      (Energy|StartUsingEnergy _energycomponent 0.0 (/ _secondary_attack_cost (Variables|Default|GetEnergy%20Use%20Duration)) (Variables|Default|GetEnergy%20Use%20Identifier) 0))))

(fn StopUsingEnergy
  (bind _energycomponent (Variables|Default|GetEnergyComponent)
    (:then
      (Energy|StopUsingEnergy _energycomponent (Variables|Default|GetEnergy%20Use%20Identifier) 0))
    (:else)))
