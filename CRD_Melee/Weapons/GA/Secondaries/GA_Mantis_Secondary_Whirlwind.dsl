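; Mantis secondary melee ability that performs a multi-hit whirlwind attack, gated by a hotfix toggle.
; Montage notifies count spins and trigger async targeting that applies feedback and effects to hits, jumping to the Finish section at the target count.
; Spin completion and counted targeting results feed wait-for-both gates and a net sync before ending the ability.
; The damage spec override scales the weapon's secondary damage by a per-hit modifier.
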(event Custom|Targeting
  (bind _traceoptions (Variables|Targeting|GetTraceOptions))
  ;[1 1 1 1 64] Perform targeting
  (Utilities|IsValid _traceoptions
    (:Is%20Valid
      (bind (_asynctaskproxy _targetinghandle) (Class|AsyncActionPerformTargeting|PerformTargetingAsyncAction (CallFunction|GetAvatarasFortPlayerPawn) _traceoptions true)
        (:Targeted
          (if (Math|Conversions|ToBoolean(Integer) (Utilities|Array|Length (TargetingSystem|TargetingResults|GetTargetingResults "/Script/TargetingSystem.Default__TargetingSubsystem" _targetinghandle)))
            (CallFunction|ApplySecondaryFeedback)
            (for _array_element (TargetingSystem|TargetingResults|GetTargetingResults "/Script/TargetingSystem.Default__TargetingSubsystem" _targetinghandle)
              (CallFunction|ApplyEffectstoTargets (Ability|TargetData|AbilityTargetDatafromHitResult _array_element)))
            ;;[0 1 0.02679] Match applications of targeting to the whirlwind target so we are targeting the correct number of timers, per activation of the ability
            (bind _result (Math|Integer|IncrementInt (Variables|AbilityRuntimeData|GetApplicationCount))
              (:%20%20
                (bind _output_get (Variables|AbilityRuntimeData|SetApplicationCount _result))
                (if (>= _output_get (Variables|AbilityRuntimeData|GetWhirldwindTargetCount))
                  (Utilities|FlowControl|WaitforBothExecutions true
                    (:Then
                      (Ability|Tasks|WaitNetSync "BothWait"
                        (:OnSync
                          (Ability|EndAbility))))))))))))))

(event Custom|CompletedSpin
  (Utilities|FlowControl|WaitforBothExecutions true
    (:Then
      (Ability|Tasks|WaitNetSync "BothWait"
        (:OnSync
          (Ability|EndAbility))))))

(event Ability|EventActivateAbility
  (bind (_montageinfo_montagetoplay _montageinfo_animplayrate _montageinfo_animrootmotiontranslationscale _montageinfo_montagesectiontoplay _montageinfo_overridesection) (Utilities|Struct|BreakFortGameplayAbilityMontageInfo (Variables|GetMontageInfo_MontageToPlay)))
  ;[1 1 1 1 64] Activation and whirlwind target tracking
  (if (ScalableFloat|GetValueasBool (Variables|Default|GetWhirlwindAttackEnabledHotfixRow))
    (|CallParentFunction)
    (Variables|AbilityRuntimeData|SetApplicationCount 0)
    (Variables|AbilityRuntimeData|SetWhirlwindCount 0)
    (Ability|Tasks|PlayMontageAndWaitNotify "None" _montageinfo_montagetoplay _montageinfo_animplayrate _montageinfo_overridesection true _montageinfo_animrootmotiontranslationscale
      (:then
        (Ability|ExecuteGameplayCueOnOwner (Variables|Tags|GetGCN_Whirlwind_Start) (Class|AthenaFunctionLibrary|MakeSimpleFortEffectContext (Ability|GetAvatarActorfromActorInfo) (Ability|GetAvatarActorfromActorInfo))))
      (:OnNotifyBegin
        (bind _result (Math|Integer|IncrementInt (Variables|AbilityRuntimeData|GetWhirlwindCount))
          (:%20%20
            (bind _output_get (Variables|AbilityRuntimeData|SetWhirlwindCount _result))
            (if (>= _output_get (Variables|AbilityRuntimeData|GetWhirldwindTargetCount))
              (Utilities|FlowControl|DoOnce false
                (:Completed
                  (CallFunction|Targeting)
                  (Ability|Tasks|SetNextMontageSectionandWait (CallFunction|GetMontageAsset) "Finish"
                    (:then
                      (Ability|ExecuteGameplayCueOnOwner (Variables|Tags|GetGCN_Whirlwind_End) (Class|AthenaFunctionLibrary|MakeSimpleFortEffectContext (Ability|GetAvatarActorfromActorInfo) (Ability|GetAvatarActorfromActorInfo))))
                    (:OnComplete
                      _)
                    (:OnInterrupted
                      _))
                  (CallFunction|CompletedSpin)))
              (else
                (CallFunction|Targeting)))))))
    (else
      (Ability|EndAbility))))

(fn GetSecondaryDamageSpec
  (params (JumpAttack :type Boolean))
  (bind _whirlwindperhitdamagemod (Variables|Default|GetWhirlwindPerHitDamageMod))
  (bind _returnvalue (ScalableFloat|GetValueatLevel _whirlwindperhitdamagemod))
  (bind (_abilitytag _secondary_damage) (CreativeMeleeWeaponInterface|GetSecondaryAbility(Message) (Variables|Default|GetWeapon)))
  (bind _returnvalue_2 (Ability|GameplayEffect|AssignTagSetbyCallerMagnitude (|CallParentFunction JumpAttack) "(TagName=\"SetByCaller.MeleeManager.BaseDamage\")" (* _secondary_damage _returnvalue)))
  (Ability|GameplayEffect|AssignTagSetbyCallerMagnitude :DataTag "(TagName=\"SetByCaller.MeleeManager.EnvironmentDamage\")" :Magnitude (* _secondary_damage _returnvalue))
  (return _returnvalue_2))
